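; Anim notify state that hides the Chain Whip's melee trail for the span of an animation window.
; Begin and End each validate the mesh, cast its owner to FortPlayerPawn and the current weapon to B_ChainWhip.
; Begin calls SetHiddenMeleeTrail with true to hide the trail, and End calls it with false to show it again.
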(fn Received_NotifyBegin
  (params (MeshComp :type Skeletal%20Mesh%20Component%20Object%20Reference) (Animation :type Anim%20Sequence%20Base%20Object%20Reference) (TotalDuration :type Float%20%28single-precision%29) (EventReference :type Anim%20Notify%20Event%20Reference%20Structure%20%28by%20ref%29))
  ;[0.03341 0.1719 0.05714] NotifyBegin: Hide the melee trail component on the weapon.
  (Utilities|FlowControl|Sequence
    (:then_0
      (Utilities|IsValid MeshComp
        (:Is%20Valid
          (bind _asfort_player_pawn (Utilities|Casting|CastToFortPlayerPawn (Components|GetOwner MeshComp)))
          (bind _asb_chain_whip (Utilities|Casting|CastToB_ChainWhip (Weapon|GetCurrentWeapon _asfort_player_pawn)))
          (Class|BChainWhip|SetHiddenMeleeTrail _asb_chain_whip true)
          (return true))))
    (:then_1
      (return false))))

(fn Received_NotifyEnd
  (params (MeshComp :type Skeletal%20Mesh%20Component%20Object%20Reference) (Animation :type Anim%20Sequence%20Base%20Object%20Reference) (EventReference :type Anim%20Notify%20Event%20Reference%20Structure%20%28by%20ref%29))
  ;[0.02994 0.1037 0.125] NotifyEnd: Unhide the melee trail component on the weapon.
  (Utilities|FlowControl|Sequence
    (:then_0
      (Utilities|IsValid MeshComp
        (:Is%20Valid
          (bind _asfort_player_pawn (Utilities|Casting|CastToFortPlayerPawn (Components|GetOwner MeshComp)))
          (bind _asb_chain_whip (Utilities|Casting|CastToB_ChainWhip (Weapon|GetCurrentWeapon _asfort_player_pawn)))
          (Class|BChainWhip|SetHiddenMeleeTrail _asb_chain_whip false)
          (return true))))
    (:then_1
      (return false))))
