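; Anim notify state for the Chain Whip that applies a configurable camera modifier to the player for the duration of the notify window.
; On begin, it casts the mesh owner to a Fort player pawn and, if locally controlled, finds or adds the modifier through WeaponCameraFunctionLibrary.
; On end, the same find-or-add call retrieves the modifier instance, which is then removed from that player's camera manager.
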(fn Received_NotifyBegin
  (params (MeshComp :type Skeletal%20Mesh%20Component%20Object%20Reference) (Animation :type Anim%20Sequence%20Base%20Object%20Reference) (TotalDuration :type Float%20%28single-precision%29) (EventReference :type Anim%20Notify%20Event%20Reference%20Structure%20%28by%20ref%29))
  ;[0.02096 0.151 0.02562] ReceivedNotfiyBegin: Apply the camera modifier to the player.
  (Utilities|FlowControl|Sequence
    (:then_0
      (Utilities|IsValid MeshComp
        (:Is%20Valid
          (bind _asfort_player_pawn (Utilities|Casting|CastToFortPlayerPawn (Components|GetOwner MeshComp)))
          (if (Pawn|IsLocallyControlled _asfort_player_pawn)
            (Class|WeaponCameraFunctionLibrary|FindorAddCameraModifierforPawnNative (Variables|Default|GetCameraModifier) _asfort_player_pawn)))))
    (:then_1
      (return false))))

(fn Received_NotifyEnd
  (params (MeshComp :type Skeletal%20Mesh%20Component%20Object%20Reference) (Animation :type Anim%20Sequence%20Base%20Object%20Reference) (EventReference :type Anim%20Notify%20Event%20Reference%20Structure%20%28by%20ref%29))
  ;[0.09896 0.05799 0.01674] ReceivedNotifyEnd: Remove the camera modifier.
  (Utilities|FlowControl|Sequence
    (:then_0
      (Utilities|IsValid MeshComp
        (:Is%20Valid
          (bind _asfort_player_pawn (Utilities|Casting|CastToFortPlayerPawn (Components|GetOwner MeshComp)))
          (if (Pawn|IsLocallyControlled _asfort_player_pawn)
            (Utilities|IsValid (Fortnite|GetFortPlayerControllerfromActor _asfort_player_pawn)
              (:Is%20Valid
                (bind _playercameramanager (Variables|/Script/Engine.PlayerController|GetPlayerCameraManager (Fortnite|GetFortPlayerControllerfromActor _asfort_player_pawn))
                  (:then
                    (bind _outmodifier (Class|WeaponCameraFunctionLibrary|FindorAddCameraModifierforPawnNative (Variables|Default|GetCameraModifier) _asfort_player_pawn))
                    (Game|Player|RemoveCameraModifier _playercameramanager _outmodifier))
                  (:else))))))))
    (:then_1
      (return false))))
