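; Anim notify for the chain whip that forces a net update on the owning actor, the player pawn, when it fires.
; This makes the montage section being played reach simulated player pawns promptly.
; It validates the mesh component and its owner, then calls ForceNetUpdate only where the owner has authority, returning true in that case and false otherwise.
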(fn Received_Notify
  (params (MeshComp :type Skeletal%20Mesh%20Component%20Object%20Reference) (Animation :type Anim%20Sequence%20Base%20Object%20Reference) (EventReference :type Anim%20Notify%20Event%20Reference%20Structure%20%28by%20ref%29))
  (bind _returnvalue (Components|GetOwner MeshComp))
  ;[0.0239 0.1314 0.1354] ReceivedNotify: ForceNetUpdate the owning actor (player pawn) when this notify is received so that the Section of the Montage being played is updated on the simulated player pawns in the game in a timely manner.
  (Utilities|FlowControl|Sequence
    (:then_0
      (Utilities|IsValid MeshComp
        (:Is%20Valid
          (Utilities|IsValid _returnvalue
            (:Is%20Valid
              (if (Networking|HasAuthority _returnvalue)
                (Networking|ForceNetUpdate _returnvalue)
                (return true)))))))
    (:then_1
      (return false))))
