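; Anim notify for the Chain Whip weapon that refreshes the attack range on the animation Blueprint when the notify fires.
; On non-server machines only, it casts the mesh owner to a Fort player pawn and its current weapon to B_ChainWhip.
; It then calls UpdateAttackRangeAnimBP on that weapon, so the range logic itself lives in the weapon Blueprint.
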(fn Received_Notify
  (params (MeshComp :type Skeletal%20Mesh%20Component%20Object%20Reference) (Animation :type Anim%20Sequence%20Base%20Object%20Reference) (EventReference :type Anim%20Notify%20Event%20Reference%20Structure%20%28by%20ref%29))
  ;[0.07949 0.1823 0.06205] ReceivedNotify: When notify is received, have the weapon update the attack range on the anim BP.
  ; Only needs to be done on the client.
  (Utilities|FlowControl|Sequence
    (:then_0
      (if (not (Networking|IsServer))
        (Utilities|IsValid MeshComp
          (:Is%20Valid
            (bind _asfort_player_pawn (Utilities|Casting|CastToFortPlayerPawn (Components|GetOwner MeshComp)))
            (bind _asb_chain_whip (Utilities|Casting|CastToB_ChainWhip (Weapon|GetCurrentWeapon _asfort_player_pawn)))
            (Class|BChainWhip|UpdateAttackRangeAnimBP _asb_chain_whip)
            (return true)))))
    (:then_1
      (return false))))
