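# Item animation layer Blueprint for the Fortnite player skeleton, supplying Hades chain whip poses for upper body and aiming.
# When bIsUpperBodyReplacement is set, ItemUpperBody replaces the cached input with either a non-targeted pose or a sliding pose plus additive loop, selected by bIsMotorcycle.
# ItemFullBodyAimOffsets applies the Hades aim offset, driven by AimPitch, when bIsHadesAO is set, otherwise passing the input through.
# All other layer outputs are unwired.
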AnimBlueprint /SunRoseChainWhipGameplay/Items/ChainWhip/Animations/Hades_Player_LayerAnimBP
target_skeleton=&Fortnite_M_Avg_Player_Skeleton
graphs [
 AnimGraph {nodes [Output Pose() {layerGroup=None}]}
 EventGraph {links [], nodes []}
 ItemAimOffset {nodes [
   Input Pose!() -> (Pose, AimOffsetAlpha:double, Yaw:double, Pitch:double) {name=InPose_AimOffset}
   Input Pose!() -> (Pose) {name=InPose_PreAimOffsets}
   Output Pose() {layerGroup=Item Override}
  ]}
 ItemFallAdditive {nodes [
   Input Pose() {name=InPoseItemFallAdditive}
   Output Pose() {layerGroup=Item Override}
  ]}
 ItemFallLandAdditive {nodes [
   Input Pose() {name=InPoseItemFallLandAdditive}
   Output Pose() {layerGroup=Item Override}
  ]}
 ItemFallLevel2Override {nodes [
   Input Pose() {name=InPose_FallLevel2}
   Output Pose() {layerGroup=Item Override}
  ]}
 ItemFinalPoseOverride {nodes [
   Output Pose() {layerGroup=Item Override}
   Save cached pose 'SavedPose18730'(Pose=Input Pose() {name=InFinalPose}) {cacheName=SavedPose18730}
   Use cached pose 'SavedPose18730'()
  ]}
 ItemFullBodyAimOffsets {nodes [
   Input Pose() {name=InPosePreFullBodyAimOffsets}
   n1 = Use cached pose 'SavedPose24996'()
   n2 = Get AimPitch()
   Output Pose(Result=Blend Poses by bool(
     BlendPose_0=AimOffset Player 'Hades_Player_AO'!(X:float=@n2, BasePose:Pose Link Structure=@n1, Alpha:float=1) -> (Pose:Pose Link Structure) {blendSpace=&/Hades_Player_AO, alphaBoolBlend {blendOption=Linear}}
     BlendPose_1=Use cached pose 'SavedPose24996'()
     BlendTime_1=2
     bActiveValue=Get bIsHadesAO() -> (bIsHadesAO:bool)
    ) {blendType=Linear}) {layerGroup=Item Override}
   Save cached pose 'SavedPose24996'(Pose=Input Pose!() -> (Pose, FullBody_AimOffsetAlpha:double, Aim_Yaw:double, Aim_Pitch:double) {name=InPoseFullBodyAimOffsets}) {cacheName=SavedPose24996}
  ]}
 ItemFullBodyOverride {nodes [
   Input Pose!() -> (Pose) {name=InPoseFullBody}
   Output Pose() {layerGroup=Item Override}
  ]}
 ItemIdleOverride {nodes [
   Input Pose() {name=InPose_Idle}
   Output Pose() {layerGroup=Item Override}
  ]}
 ItemJetPackJumpAdditive {nodes [Output Pose() {layerGroup=Item Override}]}
 ItemJetPackStartAdditive {nodes [Output Pose() {layerGroup=Item Override}]}
 ItemJumpLoopAdditive {nodes [
   Input Pose() {name=InPoseItemJumpLoopAdditive}
   Output Pose() {layerGroup=Item Override}
  ]}
 ItemJumpOverride {nodes [
   Input Pose() {name=InPoseJump}
   Output Pose() {layerGroup=Item Override}
  ]}
 ItemJumpUpAdditive {nodes [
   Input Pose() {name=InPoseItemJumpUpAdditive}
   Output Pose() {layerGroup=Item Override}
  ]}
 ItemPreAimOffsets {nodes [
   Input Pose() {name=InPosePreAimOffsets}
   Output Pose() {layerGroup=Item Override}
  ]}
 ItemPreIK {nodes [
   Input Pose() {name=InPose_PreIK}
   Output Pose() {layerGroup=Item Override}
  ]}
 ItemRightHandIK {nodes [
   Input Pose() {name=InPosePreRightHandIK}
   Input Pose() {name=InPosePostRightHandIK}
   Output Pose() {layerGroup=Item Override}
  ]}
 ItemSkeletalControl {nodes [
   Input Pose()
   Output Pose() {layerGroup=Item Override}
  ]}
 ItemSwimJumpFallLoop {nodes [Output Pose() {layerGroup=Item Override}]}
 ItemSwimJumpSurfaceEnd {nodes [Output Pose() {layerGroup=Item Override}]}
 ItemSwimJumpSurfaceLoop {nodes [Output Pose() {layerGroup=Item Override}]}
 ItemUpperBody {nodes [
   n1 = Apply Additive(
    Base=Sequence Player 'Hades_Sliding_Pose'() {bLoopAnimation=false, sequence=&/Hades_Sliding_Pose}
    Additive="Sequence Player (additive) 'Hades_Player_Slide_Loop'"() {sequence=&/Hades_Player_Slide_Loop}
   ) {alphaBoolBlend {blendOption=Linear}}
   n2 = Sequence Player 'Hades_Player_NonTargeted'() {bLoopAnimation=false, sequence=&/Hades_Player_NonTargeted}
   n3 = Get bIsMotorcycle() -> (bIsMotorcycle:bool)
   Output Pose(Result=Blend Poses by bool(
     BlendPose_0=Blend Poses by bool(BlendPose_0=@n2, BlendPose_1=@n1, bActiveValue=@n3) {blendType=Linear}
     BlendPose_1=Use cached pose 'SavedPose31167'()
     bActiveValue=Get bIsUpperBodyReplacement() -> (bIsUpperBodyReplacement:bool)
    ) {blendType=Linear}) {layerGroup=Item Override}
   Save cached pose 'SavedPose31167'(Pose=Input Pose!() -> (Pose, InputParam:Fort Anim Input Adjusted Aim Structure) {name=InPoseUpperBody}) {cacheName=SavedPose31167}
  ]}
 ItemZipLineStartAdditive {nodes [Output Pose() {layerGroup=Item Override}]}
]
variables []
