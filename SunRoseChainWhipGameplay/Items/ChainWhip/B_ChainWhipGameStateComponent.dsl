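; Server-side game state component that tallies Chain Whip hookshot analytics for a match.
; Notify functions count human-instigated throws and hooks (pawns by type and team, pickups, floppers, vehicles).
; SubmitAnalytics fires once, outside replays, at EndGame or editor EndPlay, sending "Core.SunRose.ChainWhip".
; Quirks: the throw count reuses the "TotalNumOfSuccessfulHookshots" attribute name, and average distance sums only participant-target hooks but divides by all pawn hooks.
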(event EventBeginPlay
  ;[0.02436 0.1042 0.04578] BeginPlay: Bind to know when the game phase changes. Used to fire analytics at the end of the match.
  (|CallParentFunction)
  (Utilities|SwitchNetMode
    (:Server
      (bind _gamestate (Class|FortAsyncActionCurrentPlaylistReady|PlaylistDataReadyAsync)
        (:OnReady
          (Utilities|IsValid _gamestate
            (:Is%20Valid
              (bind _returnvalue (GameplayEventRouter|GetEventRouter "/Script/FortniteGame.FortGameState" _gamestate))
              (Utilities|IsValid _returnvalue
                (:Is%20Valid
                  (bind (_asynctaskproxy _payload) (GameplayEventRouter|StartListeningToStatefulEvent _returnvalue "/Script/FortniteGame.FortGamePhaseUpdatedEvent")
                    (:then
                      (Variables|Transient|SetGamePhaseUpdatedAsyncTask _asynctaskproxy))
                    (:OnEventReceived
                      _)
                    (:OnSavedEventStateExists
                      _))
                  (Protected|HandleGamePhaseChanged (Utilities|Struct|BreakFortGamePhaseUpdatedEvent _payload)))
                (:Is%20Not%20Valid
                  ;;[1 0 0.03209] Warning: EventRouter is null.
                  (Fortnite|FortShippingLog self "BeginPlay: EventRouter is null." true))))))))))

(event EventEndPlay
  (params (EndPlayReason :type EEndPlayReason%20Enum :default "Destroyed"))
  ;[0.1302 0.09219 0.02218] EndPlay: Clean up binds and fire analytics logic if playing in editor.
  (Utilities|FlowControl|Sequence
    (:then_0
      (Utilities|SwitchNetMode
        (:Server
          (if (Editor|InEditororPIE)
            (Protected|SubmitAnalytics)
            (else
              _))
          (bind _gamephaseupdatedasynctask (Variables|Transient|GetGamePhaseUpdatedAsyncTask)
            (:then
              (GameplayEventRouter|StopListeningtoStatefulEvent _gamephaseupdatedasynctask))
            (:else)))))
    (:then_1
      (|CallParentFunction EndPlayReason))))

(fn HandleGamePhaseChanged
  (params (GamePhase :type EAthenaGamePhase%20Enum))
  ;[0.0164 0.07935 0.1198] HandleGamePhaseChanged: Once we're in EndGame, submit the analytics.
  (Utilities|SwitchNetMode
    (:Server
      (if (Utilities|Enum|Equal(Enum) GamePhase "EndGame")
        (Protected|SubmitAnalytics)
        (bind _gamephaseupdatedasynctask (Variables|Transient|GetGamePhaseUpdatedAsyncTask)
          (:then
            (GameplayEventRouter|StopListeningtoStatefulEvent _gamephaseupdatedasynctask))
          (:else))))))

(fn IsAthenaGameParticipant
  (params (PlayerPawn :type Fort%20Player%20Pawn%20Object%20Reference))
  ;[0.03566 0.1146 0.11] IsHumanPlayerPawn: Returns if the input player pawn is a game participant.
  (Utilities|FlowControl|Sequence
    (:then_0
      (Utilities|IsValid PlayerPawn
        (:Is%20Valid
          (bind _playerstate (Variables|/Script/Engine.Pawn|GetPlayerState PlayerPawn)
            (:then
              (bind _asfort_player_state_athena (Utilities|Casting|CastToFortPlayerStateAthena _playerstate))
              (return (Class|FortPlayerStateAthena|IsAnAthenaGameParticipant _asfort_player_state_athena)))
            (:else)))))
    (:then_1
      (return false))))

(fn IsHumanPlayerPawn
  (params (PlayerPawn :type Fort%20Player%20Pawn%20Object%20Reference))
  ;[0.03566 0.1146 0.11] IsHumanPlayerPawn: Returns if the input player pawn is controlled by a player.
  (Utilities|FlowControl|Sequence
    (:then_0
      (Utilities|IsValid PlayerPawn
        (:Is%20Valid
          (Utilities|IsValid (Fortnite|GetFortPlayerControllerfromActor PlayerPawn)
            (:Is%20Valid
              (return true))))))
    (:then_1
      (return false))))

(fn NotifyFlopperHooked
  (params (PlayerPawn :type Fort%20Player%20Pawn%20Object%20Reference))
  ;[0.02592 0.09375 0.08627] NotifyFloppedHooked: Track the number of floppers hooked.
  (Utilities|SwitchNetMode
    (:Server
      (Utilities|IsValid PlayerPawn
        (:Is%20Valid
          (bind _breturn (Helper|IsHumanPlayerPawn PlayerPawn))
          (if _breturn
            (bind _breturn_2 (Helper|IsAthenaGameParticipant PlayerPawn))
            (if _breturn_2
              (Math|Integer|IncrementInt (Variables|Analytics|Hookshot|GetNumOfParticipantPlayerPawnsFloppersHooked)))))))))

(fn NotifyHookshotThrown
  (params (InstigatorPlayerPawn :type Fort%20Player%20Pawn%20Object%20Reference))
  ;[0.01697 0.142 0.151] NotifyHookshotThrown: Track the number of participating human players that have thrown the hookshot.
  (Utilities|SwitchNetMode
    (:Server
      (bind _breturn (Helper|IsHumanPlayerPawn InstigatorPlayerPawn))
      (if _breturn
        (bind _breturn_2 (Helper|IsAthenaGameParticipant InstigatorPlayerPawn))
        (if _breturn_2
          (Math|Integer|IncrementInt (Variables|Analytics|Hookshot|GetNumOfParticipantPlayerPawnsHookshotsThrown)))))))

(fn NotifyPickupsHooked
  (params (PlayerPawn :type Fort%20Player%20Pawn%20Object%20Reference) (NumOfPickups :type Integer))
  ;[0.02501 0.2188 0.2144] NotifyPickupsHooked: Track the number of pickups that are hooked.
  (Utilities|SwitchNetMode
    (:Server
      (Utilities|IsValid PlayerPawn
        (:Is%20Valid
          (bind _breturn (Helper|IsHumanPlayerPawn PlayerPawn))
          (if _breturn
            (bind _breturn_2 (Helper|IsAthenaGameParticipant PlayerPawn))
            (if _breturn_2
              (Variables|Analytics|Hookshot|SetNumOfParticipantPlayerPawnsPickupsHooked (+ (Variables|Analytics|Hookshot|GetNumOfParticipantPlayerPawnsPickupsHooked) (Math|Integer|Max(Integer) NumOfPickups))))))))))

(fn NotifySuccessfulHookshot
  (params (InstigatorPlayerPawn :type Fort%20Player%20Pawn%20Object%20Reference) (TargetCharacter :type Character%20Object%20Reference) (HookshotProjectile :type B%20Prj%20Chain%20Whip%20Hookshot%20V%202%20Object%20Reference))
  ;[0.02561 0.128 0.1406] NotifySuccessfulHookshot: Track the hookshot for analytics.
  ; Called from GA_ChainWhip_Hookshot_Pull.
  (Utilities|SwitchNetMode
    (:Server
      (Utilities|IsValid InstigatorPlayerPawn
        (:Is%20Valid
          (Utilities|IsValid TargetCharacter
            (:Is%20Valid
              (bind _breturn (Helper|IsHumanPlayerPawn InstigatorPlayerPawn)) ; Only care about human instigators.
              (if _breturn
                (LocalVariables|SetbSameTeam (Team|OnSameTeam InstigatorPlayerPawn TargetCharacter))
                (Utilities|SwitchCharacterType TargetCharacter
                  (:FortPlayerPawn
                    (bind _breturn_2 (Helper|IsAthenaGameParticipant (Utilities|Casting|CastToFortPlayerPawn (Variables|GetTargetCharacter))))
                    (if _breturn_2
                      (if (LocalVariables|GetbSameTeam)
                        (Math|Integer|IncrementInt (Variables|Analytics|Hookshot|GetNumOfParticipantPlayerPawnsHookedOnSameTeam)
                          (:%20%20
                            _))
                        (else
                          (Math|Integer|IncrementInt (Variables|Analytics|Hookshot|GetNumOfParticipantPlayerPawnsHookedOnOtherTeam)
                            (:%20%20
                              _))))
                      (bind _length (Class|BPrjHookshot|GetCurrentRopeLength HookshotProjectile))
                      (Variables|Analytics|Hookshot|SetTotalSuccessfulHookshotDistance (+ _length (Variables|Analytics|Hookshot|GetTotalSuccessfulHookshotDistance)))
                      (else
                        (if (LocalVariables|GetbSameTeam)
                          (Math|Integer|IncrementInt (Variables|Analytics|Hookshot|GetNumOfNonParticipantPlayerPawnsHookedOnSameTeam)
                            (:%20%20
                              _))
                          (else
                            (Math|Integer|IncrementInt (Variables|Analytics|Hookshot|GetNumOfNonParticipantPlayerPawnsHookedOnOtherTeam)
                              (:%20%20
                                _)))))))
                  (:FortAIPawn
                    (if (LocalVariables|GetbSameTeam)
                      (Math|Integer|IncrementInt (Variables|Analytics|Hookshot|GetNumOfAIPawnsHookedOnSameTeam)
                        (:%20%20
                          _))
                      (else
                        (Math|Integer|IncrementInt (Variables|Analytics|Hookshot|GetNumOfAIPawnsHookedOnOtherTeam)
                          (:%20%20
                            _)))))) ; Switching on target type
              ))))))))

(fn NotifyVehicleHooked
  (params (PlayerPawn :type Fort%20Player%20Pawn%20Object%20Reference))
  ;[0.01974 0.1458 0.1444] NotifyVehicleHooked: Track the number of vehicles hooked.
  (Utilities|SwitchNetMode
    (:Server
      (Utilities|IsValid PlayerPawn
        (:Is%20Valid
          (bind _breturn (Helper|IsHumanPlayerPawn PlayerPawn))
          (if _breturn
            (bind _breturn_2 (Helper|IsAthenaGameParticipant PlayerPawn))
            (if _breturn_2
              (Math|Integer|IncrementInt (Variables|Analytics|Hookshot|GetNumOfParticipantPlayerPawnsVehiclesHooked)))))))))

(fn SubmitAnalytics
  (bind _numofaipawnshookedonotherteam (Variables|Analytics|Hookshot|GetNumOfAIPawnsHookedOnOtherTeam))
  (bind _numofaipawnshookedonsameteam (Variables|Analytics|Hookshot|GetNumOfAIPawnsHookedOnSameTeam))
  (bind _numofnonparticipantplayerpawnshookedonotherteam (Variables|Analytics|Hookshot|GetNumOfNonParticipantPlayerPawnsHookedOnOtherTeam))
  (bind _numofnonparticipantplayerpawnshookedonsameteam (Variables|Analytics|Hookshot|GetNumOfNonParticipantPlayerPawnsHookedOnSameTeam))
  (bind _numofparticipantplayerpawnshookedonotherteam (Variables|Analytics|Hookshot|GetNumOfParticipantPlayerPawnsHookedOnOtherTeam))
  (bind _numofparticipantplayerpawnshookedonsameteam (Variables|Analytics|Hookshot|GetNumOfParticipantPlayerPawnsHookedOnSameTeam))
  (bind _returnvalue (Utilities|Operators|Add _numofparticipantplayerpawnshookedonsameteam _numofparticipantplayerpawnshookedonotherteam _numofnonparticipantplayerpawnshookedonsameteam _numofnonparticipantplayerpawnshookedonotherteam _numofaipawnshookedonsameteam _numofaipawnshookedonotherteam))
  (bind _returnvalue_2 (Math|Conversions|ToFloat(Integer) _returnvalue))
  (bind _totalsuccessfulhookshotdistance (Variables|Analytics|Hookshot|GetTotalSuccessfulHookshotDistance))
  (bind _returnvalue_3 (Math|Float|SafeDivide _totalsuccessfulhookshotdistance _returnvalue_2))
  (bind _numofaipawnshookedonotherteam_2 (Variables|Analytics|Hookshot|GetNumOfAIPawnsHookedOnOtherTeam))
  (bind _numofaipawnshookedonsameteam_2 (Variables|Analytics|Hookshot|GetNumOfAIPawnsHookedOnSameTeam))
  (bind _numofnonparticipantplayerpawnshookedonotherteam_2 (Variables|Analytics|Hookshot|GetNumOfNonParticipantPlayerPawnsHookedOnOtherTeam))
  (bind _numofnonparticipantplayerpawnshookedonsameteam_2 (Variables|Analytics|Hookshot|GetNumOfNonParticipantPlayerPawnsHookedOnSameTeam))
  (bind _numofparticipantplayerpawnshookedonotherteam_2 (Variables|Analytics|Hookshot|GetNumOfParticipantPlayerPawnsHookedOnOtherTeam))
  (bind _numofparticipantplayerpawnshookedonsameteam_2 (Variables|Analytics|Hookshot|GetNumOfParticipantPlayerPawnsHookedOnSameTeam))
  (bind _returnvalue_4 (Utilities|Operators|Add _numofparticipantplayerpawnshookedonsameteam_2 _numofparticipantplayerpawnshookedonotherteam_2 _numofnonparticipantplayerpawnshookedonsameteam_2 _numofnonparticipantplayerpawnshookedonotherteam_2 _numofaipawnshookedonsameteam_2 _numofaipawnshookedonotherteam_2))
  (bind _numofparticipantplayerpawnshookedonotherteam_3 (Variables|Analytics|Hookshot|GetNumOfParticipantPlayerPawnsHookedOnOtherTeam))
  (bind _numofnonparticipantplayerpawnshookedonsameteam_3 (Variables|Analytics|Hookshot|GetNumOfNonParticipantPlayerPawnsHookedOnSameTeam))
  (bind _numofparticipantplayerpawnshookedonsameteam_3 (Variables|Analytics|Hookshot|GetNumOfParticipantPlayerPawnsHookedOnSameTeam))
  (bind _numofaipawnshookedonsameteam_3 (Variables|Analytics|Hookshot|GetNumOfAIPawnsHookedOnSameTeam))
  (bind _numofaipawnshookedonotherteam_3 (Variables|Analytics|Hookshot|GetNumOfAIPawnsHookedOnOtherTeam))
  (bind _numofparticipantplayerpawnshookshotsthrown (Variables|Analytics|Hookshot|GetNumOfParticipantPlayerPawnsHookshotsThrown))
  (bind _numofparticipantplayerpawnsfloppershooked (Variables|Analytics|Hookshot|GetNumOfParticipantPlayerPawnsFloppersHooked))
  (bind _numofparticipantplayerpawnsvehicleshooked (Variables|Analytics|Hookshot|GetNumOfParticipantPlayerPawnsVehiclesHooked))
  (bind _numofnonparticipantplayerpawnshookedonotherteam_3 (Variables|Analytics|Hookshot|GetNumOfNonParticipantPlayerPawnsHookedOnOtherTeam))
  (bind _numofparticipantplayerpawnspickupshooked (Variables|Analytics|Hookshot|GetNumOfParticipantPlayerPawnsPickupsHooked))
  ;[0.03754 0.005307 0.04167] SubmitAnalytics: Fire analytics if it hasn't been done yet.
  (Utilities|SwitchNetMode
    (:Server
      (if (and (not (Replay|GetIsPlayingReplay)) (not (Variables|Transient|GetbAnalyticsSubmitted)))
        (Variables|Transient|SetbAnalyticsSubmitted true)
        (Utilities|FlowControl|Sequence
          (:then_0
            (Utilities|FlowControl|Sequence
              (:then_0
                ;;[1 1 1] NumOfParticipantPlayerPawnsHookedOnSameTeam
                (bind _returnvalue_5 (Analytics|ConstructAnalyticsAttribute :AttributeName "NumOfParticipantPlayerPawnsHookedOnSameTeam" :IntValue _numofparticipantplayerpawnshookedonsameteam_3 :floatValue -1.0 :BoolValue false))
                (Utilities|Array|Add (LocalVariables|GetAnalyticsEventAttributes) _returnvalue_5)
                (Fortnite|FortShippingLog self (Utilities|String|Append "NumOfParticipantPlayerPawnsHookedOnSameTeam: " (Utilities|String|ToString(Integer) _numofparticipantplayerpawnshookedonsameteam_3)))
                ;;[1 1 1] NumOfParticipantPlayerPawnsHookedOnOtherTeam
                (bind _returnvalue_6 (Analytics|ConstructAnalyticsAttribute :AttributeName "NumOfParticipantPlayerPawnsHookedOnOtherTeam" :IntValue _numofparticipantplayerpawnshookedonotherteam_3 :floatValue -1.0 :BoolValue false))
                (Utilities|Array|Add (LocalVariables|GetAnalyticsEventAttributes) _returnvalue_6)
                (Fortnite|FortShippingLog self (Utilities|String|Append "NumOfParticipantPlayerPawnsHookedOnOtherTeam: " (Utilities|String|ToString(Integer) _numofparticipantplayerpawnshookedonotherteam_3)))
                ;;[1 1 1] NumOfNonParticipantPlayerPawnsHookedOnSameTeam
                (bind _returnvalue_7 (Analytics|ConstructAnalyticsAttribute :AttributeName "NumOfNonParticipantPlayerPawnsHookedOnSameTeam" :IntValue _numofnonparticipantplayerpawnshookedonsameteam_3 :floatValue -1.0 :BoolValue false))
                (Utilities|Array|Add (LocalVariables|GetAnalyticsEventAttributes) _returnvalue_7)
                (Fortnite|FortShippingLog self (Utilities|String|Append "NumOfNonParticipantPlayerPawnsHookedOnSameTeam: " (Utilities|String|ToString(Integer) _numofnonparticipantplayerpawnshookedonsameteam_3)))
                ;;[1 1 1] NumOfNonParticipantPlayerPawnsHookedOnOtherTeam
                (bind _returnvalue_8 (Analytics|ConstructAnalyticsAttribute :AttributeName "NumOfNonParticipantPlayerPawnsHookedOnOtherTeam" :IntValue _numofnonparticipantplayerpawnshookedonotherteam_3 :floatValue -1.0 :BoolValue false))
                (Utilities|Array|Add (LocalVariables|GetAnalyticsEventAttributes) _returnvalue_8)
                (Fortnite|FortShippingLog self (Utilities|String|Append "NumOfNonParticipantPlayerPawnsHookedOnOtherTeam: " (Utilities|String|ToString(Integer) _numofnonparticipantplayerpawnshookedonotherteam_3)))
                ;;[1 1 1] NumOfAIPawnsHookedOnSameTeam
                (bind _returnvalue_9 (Analytics|ConstructAnalyticsAttribute :AttributeName "NumOfAIPawnsHookedOnSameTeam" :IntValue _numofaipawnshookedonsameteam_3 :floatValue -1.0 :BoolValue false))
                (Utilities|Array|Add (LocalVariables|GetAnalyticsEventAttributes) _returnvalue_9)
                (Fortnite|FortShippingLog self (Utilities|String|Append "NumOfAIPawnsHookedOnSameTeam: " (Utilities|String|ToString(Integer) _numofaipawnshookedonsameteam_3)))
                ;;[1 1 1] NumOfAIPawnsHookedOnOtherTeam
                (bind _returnvalue_10 (Analytics|ConstructAnalyticsAttribute :AttributeName "NumOfAIPawnsHookedOnOtherTeam" :IntValue _numofaipawnshookedonotherteam_3 :floatValue -1.0 :BoolValue false))
                (Utilities|Array|Add (LocalVariables|GetAnalyticsEventAttributes) _returnvalue_10)
                (Fortnite|FortShippingLog self (Utilities|String|Append "NumOfAIPawnsHookedOnOtherTeam: " (Utilities|String|ToString(Integer) _numofaipawnshookedonotherteam_3))))
              (:then_1
                ;;[1 1 1] TotalNumOfSuccessfulHookshots
                (bind _returnvalue_11 (Analytics|ConstructAnalyticsAttribute :AttributeName "TotalNumOfSuccessfulHookshots" :IntValue _returnvalue_4 :floatValue -1.0 :BoolValue false))
                (Utilities|Array|Add (LocalVariables|GetAnalyticsEventAttributes) _returnvalue_11)
                (Fortnite|FortShippingLog self (Utilities|String|Append "TotalNumOfSuccessfulHookshots: " (Utilities|String|ToString(Integer) _returnvalue_4)))
                ;;[1 1 1] NumOfParticipantPlayerPawnsHookshotsThrown
                (bind _returnvalue_12 (Analytics|ConstructAnalyticsAttribute :AttributeName "TotalNumOfSuccessfulHookshots" :IntValue _numofparticipantplayerpawnshookshotsthrown :floatValue -1.0 :BoolValue false))
                (Utilities|Array|Add (LocalVariables|GetAnalyticsEventAttributes) _returnvalue_12)
                (Fortnite|FortShippingLog self (Utilities|String|Append "NumOfParticipantPlayerPawnsHookshotsThrown: " (Utilities|String|ToString(Integer) _numofparticipantplayerpawnshookshotsthrown)))
                ;;[1 1 1] AverageSuccessfulHookshotDistance
                (bind _returnvalue_13 (Analytics|ConstructAnalyticsAttribute :AttributeName "AverageSuccessfulHookshotDistance" :IntValue -1 :floatValue _returnvalue_3 :BoolValue false))
                (Utilities|Array|Add (LocalVariables|GetAnalyticsEventAttributes) _returnvalue_13)
                (Fortnite|FortShippingLog self (Utilities|String|Append "AverageSuccessfulHookshotDistance: " (Utilities|String|ToString(Float) _returnvalue_3)))
                ;;[1 1 1] NumOfParticipantPlayerPawnsPickupsHooked
                (bind _returnvalue_14 (Analytics|ConstructAnalyticsAttribute :AttributeName "NumOfParticipantPlayerPawnsPickupsHooked" :IntValue _numofparticipantplayerpawnspickupshooked :floatValue -1.0 :BoolValue false))
                (Utilities|Array|Add (LocalVariables|GetAnalyticsEventAttributes) _returnvalue_14)
                (Fortnite|FortShippingLog self (Utilities|String|Append "NumOfParticipantPlayerPawnsPickupsHooked: " (Utilities|String|ToString(Integer) _numofparticipantplayerpawnspickupshooked)))
                ;;[1 1 1] NumOfParticipantPlayerPawnsFloppersHooked
                (bind _returnvalue_15 (Analytics|ConstructAnalyticsAttribute :AttributeName "NumOfParticipantPlayerPawnsFloppersHooked" :IntValue _numofparticipantplayerpawnsfloppershooked :floatValue -1.0 :BoolValue false))
                (Utilities|Array|Add (LocalVariables|GetAnalyticsEventAttributes) _returnvalue_15)
                (Fortnite|FortShippingLog self (Utilities|String|Append "NumOfParticipantPlayerPawnsFloppersHooked: " (Utilities|String|ToString(Integer) _numofparticipantplayerpawnsfloppershooked)))
                ;;[1 1 1] NumOfParticipantPlayerPawnsVehiclesHooked
                (bind _returnvalue_16 (Analytics|ConstructAnalyticsAttribute :AttributeName "NumOfParticipantPlayerPawnsVehiclesHooked" :IntValue _numofparticipantplayerpawnsvehicleshooked :floatValue -1.0 :BoolValue false))
                (Utilities|Array|Add (LocalVariables|GetAnalyticsEventAttributes) _returnvalue_16)
                (Fortnite|FortShippingLog self (Utilities|String|Append "NumOfParticipantPlayerPawnsVehiclesHooked: " (Utilities|String|ToString(Integer) _numofparticipantplayerpawnsvehicleshooked))))))
          (:then_1
            (Analytics|FireAnalyticsEventBlueprintArray "Core.SunRose.ChainWhip" (LocalVariables|GetAnalyticsEventAttributes))))))))
