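# Translucent unlit material for chain whip FX dots, drawing soft procedural circles with a two-color gradient (green by default) tinted by particle color.
# A radial sphere mask shapes both gradient and opacity, its radius driven by the Erode dynamic parameter and a reroute, apparently the panning swirl noise.
# The noise has optional UV pinching and per-particle time offsets.
# Opacity combines particle alpha with view-angle, depth and near-camera fades.
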Material /SunRoseChainWhipGameplay/Items/ChainWhip/FX/M_ChainWhip_Dots
graph {
 properties {
  bUsedWithMeshDeformer=true
  bUsedWithNiagaraSprites=true
  bUsedWithSkeletalMesh=true
  blendMode=BLEND_Translucent
  pixelDepthOffsetMode=PDOM_Legacy
  shadingModel=MSM_Unlit
  usedShadingModels=Unlit
 }
 nodes [
  n1 = TextureCoordinate()
  n2 = Constant() {r=0.5}
  n3 = LinearInterpolate(
   B=Multiply(A=DynamicParameter() {paramNames [Erode,Param2,Param3,Param4]}.Erode, B=ScalarParameter() {defaultValue=0.47, parameterName=FX_Radius})
   Alpha=NamedRerouteUsage()
  ) {constB=0.5}
  n4 = ScalarParameter() {defaultValue=0.5, parameterName=ColorLerpHardness}
  n5 = LinearInterpolate(
   A=VectorParameter() {parameterName=ColorA, defaultValue=<rgba 0.0092 0.2326 0.0487 1>}.RGB
   B=VectorParameter() {parameterName=ColorB, defaultValue=<rgba 0.1443 2 0.4297 1>}.RGB
   Alpha=Power(Base=SphereMask(A=@n1, B=@n2, Radius=@n3, Hardness=@n4), Exp=ScalarParameter() {defaultValue=5, parameterName=ColorLerpExp}) {constExponent=0.5}
  )
  n6 = ParticleColor()
  MaterialFunctionCall(Emissive=Multiply(A=@n5, B=@n6.RGB)) {materialFunction=&MF_FX_EmissiveAdjustments}
  Multiply(A=@n5, B=Divide(A=ParticleSpeed()) {constB=1000})
  n7 = MaterialFunctionCall("FadeSensitivity (optional)"=Constant() {r=1}) {materialFunction=&FadeBasedOnViewAngle}
  n8 = SphereMask(A=TextureCoordinate(), B=Constant() {r=0.5}, Radius=@n3, Hardness=LinearInterpolate(Alpha=@n3) {constA=1, constB=0.8})
  n9 = ScalarParameter() {defaultValue=1, parameterName=OverallOpacity}
  n10 = ParticleColor()
  n11 = DepthFade(
   Opacity=Multiply(A=Multiply(A=@n7.Result, B=@n8) {constB=5}, B=Multiply(A=@n9, B=@n10.A))
   FadeDistance=ScalarParameter() {defaultValue=5, parameterName=DepthFade}
  )
  Multiply(
   A=MaterialFunctionCall(Distance=Constant() {r=200}) {materialFunction=&Near_Camera_Fade}.Result
   B=Multiply(A=@n6.A, B=ShadingPathSwitch(Default=@n11, Deferred=@n11))
  )
  n12 = TextureCoordinate()
  n13 = Power(
   Base=MaterialFunctionCall(Float2=@n12) {materialFunction=&BreakOutFloat2Components}.G
   Exp=ScalarParameter() {defaultValue=1.2, parameterName=PinchAmount}
  )
  n14 = Constant() {r=1}
  n15 = LinearInterpolate(
   A=TextureCoordinate()
   B=MaterialFunctionCall(UVs=@n12, Texture Scale=AppendVector(A=@n13, B=@n14)) {materialFunction=&ScaleUVsByCenter}.UVs
   Alpha=ScalarParameter() {parameterName=UsePinchUV}
  )
  n16 = VectorParameter() {parameterName=Tile, defaultValue=<rgba 0.5 1 0 1>}
  n17 = AppendVector(A=@n16.R, B=@n16.R)
  n18 = Add(A=Time(), B=Multiply(A=ParticleRandom()) {constB=5})
  n19 = ScalarParameter() {defaultValue=1, parameterName=PanSpeed}
  n20 = TextureSampleParameter2D(UVs=Panner(Coordinate=Multiply(A=@n15, B=@n17), Time=Multiply(A=@n18, B=@n19)) {speedX=0.2, speedY=0.5}) {parameterName=Noise, samplerType=SAMPLERTYPE_LinearGrayscale, texture=/Game/Global/Textures/Default/Noise/T_SwirlNoise_N_height}
  NamedRerouteDeclaration(None=@n20.R) {name=UVDistorton, nodeColor=<rgba 0 1 0.8594 1>}
 ]
}
