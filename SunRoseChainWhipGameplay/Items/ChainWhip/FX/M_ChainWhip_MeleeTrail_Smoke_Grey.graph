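# Translucent unlit Niagara ribbon material for the smoke trail of chain whip melee attacks, tinted by particle color.
# A UV gradient mask, with a section-count parameter and a sharp fade at one edge, shapes the trail.
# Panned packed noise, tiled by a dynamic parameter, breaks up that mask: red drives depth-faded opacity, green drives brightness through a shared emissive adjustment function.
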Material /SunRoseChainWhipGameplay/Items/ChainWhip/FX/M_ChainWhip_MeleeTrail_Smoke_Grey
graph {
 properties {bUsedWithNiagaraRibbons=true, blendMode=BLEND_Translucent, pixelDepthOffsetMode=PDOM_Legacy, shadingModel=MSM_Unlit, usedShadingModels=Unlit}
 nodes [
  n1 = ParticleColor()
  n2 = MaterialFunctionCall(Float2=TextureCoordinate()) {materialFunction=&BreakOutFloat2Components}
  n3 = OneMinus(None=@n2.R)
  n4 = Divide(A=@n2.G, B=ScalarParameter() {defaultValue=1, parameterName=NumOfAnimTrailSections}) {constB=3}
  n5 = TextureCoordinate()
  n6 = Multiply(A=OneMinus(None=MaterialFunctionCall(Float2=@n5) {materialFunction=&BreakOutFloat2Components}.G)) {constB=12}
  n7 = SquareRoot(None=Multiply(A=Multiply(A=@n3, B=@n4) {constA=1}, B=Saturate(None=@n6)))
  n8 = MaterialFunctionCall(Float2=TextureCoordinate()) {materialFunction=&BreakOutFloat2Components}
  n9 = Multiply(A=@n8.R, B=Multiply(A=Constant() {r=0.4}, B=DynamicParameter() {paramNames [Param1,Tile,Param3,Param4]}.Tile))
  n10 = Multiply(A=@n8.G, B=Constant() {r=0.5})
  n11 = TextureSample(UVs=Panner(Coordinate=MaterialFunctionCall(R=@n9, G=@n10) {materialFunction=&AppendMany}.RG) {speedX=-1, speedY=0.1}) {samplerType=SAMPLERTYPE_LinearColor, texture=&T_NoisePack_LinearColor}
  n12 = LinearInterpolate(Alpha=Power(Base=Multiply(A=@n11.R) {constB=1.2}) {constExponent=3}) {constA=0.2}
  DepthFade(Opacity=Multiply(A=@n1.A, B=Multiply(A=@n7, B=@n12)))
  n13 = ScalarParameter() {defaultValue=1}
  n14 = LinearInterpolate(Alpha=Multiply(A=@n7, B=@n11.G)) {constA=0.2}
  MaterialFunctionCall(Emissive=Multiply(A=@n1.RGB, B=Multiply(A=@n13, B=@n14))) {materialFunction=&MF_FX_EmissiveAdjustments}
 ]
}
