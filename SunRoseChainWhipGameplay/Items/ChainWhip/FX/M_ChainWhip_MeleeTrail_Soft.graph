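# Soft translucent, unlit Niagara ribbon material for the chain whip's melee trail.
# Opacity combines sphere masks, U falloff and sine-shaped V profile on optionally swizzled UVs with particle alpha and depth fade.
# Emissive multiplies tint and particle color, with a Flash boost and near-camera fade term, through a shared adjustment function.
# A local-space wiggle layers time-driven sines and an optional return wiggle, scaled by WiggleAmount and masked along V.
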Material /SunRoseChainWhipGameplay/Items/ChainWhip/FX/M_ChainWhip_MeleeTrail_Soft
graph {
 properties {bUsedWithNiagaraRibbons=true, blendMode=BLEND_Translucent, pixelDepthOffsetMode=PDOM_Legacy, shadingModel=MSM_Unlit, usedShadingModels=Unlit}
 nodes [
  n1 = ObjectRadius()
  n2 = ScalarParameter() {defaultValue=2048, parameterName=MaxLength}
  Clamp(None=LinearInterpolate(Alpha=Divide(A=@n1, B=@n2)))
  n3 = NamedRerouteUsage()
  n4 = AppendVector(A=Constant() {r=2}, B=Constant() {r=0.5})
  n5 = TextureCoordinate()
  n6 = MaterialFunctionCall(XY=@n5) {materialFunction=&Swizzle}
  n7 = ScalarParameter() {parameterName=UseSwizzle}
  n8 = MaterialFunctionCall(Float2=NamedRerouteDeclaration(None=LinearInterpolate(A=@n5, B=@n6.YX, Alpha=@n7)) {name=UV, nodeColor=<rgba 0 0.1094 1 1>}) {materialFunction=&BreakOutFloat2Components}
  n9 = ParticleColor()
  n10 = Multiply(A=Power(Base=OneMinus(None=@n8.R), Exp=ScalarParameter() {defaultValue=3, parameterName=U_Exp}), B=@n9.A)
  n11 = Constant()
  n12 = Constant() {r=0.5}
  n13 = Sine(None=Multiply(A=@n8.G) {constB=0.5})
  n14 = Multiply(
   A=SphereMask(A=@n3, B=AppendVector(A=@n11, B=@n12), Radius=ScalarParameter() {defaultValue=1}) {hardnessPercent=90}
   B=Power(Base=Saturate(None=@n13), Exp=ScalarParameter() {defaultValue=2, parameterName=V_Exp})
  )
  DepthFade(
   Opacity=Multiply(A=SphereMask(A=@n3, B=@n4) {attenuationRadius=2, hardnessPercent=95}, B=Multiply(A=@n10, B=@n14))
   FadeDistance=ScalarParameter() {defaultValue=100, parameterName=DepthFadeDistance}
  ) {fadeDistanceDefault=200}
  n15 = LinearInterpolate(Alpha=ScalarParameter() {parameterName=UseNearCameraFade}) {constA=1, constB=3}
  n16 = MaterialFunctionCall(Distance=ScalarParameter() {parameterName=NearCameraFade}) {materialFunction=&Near_Camera_Fade}
  n17 = VectorParameter() {parameterName=Color, defaultValue=<rgba 1 1 1 1>}
  n18 = LinearInterpolate(A=@n17.RGB, B=Multiply(A=@n17.RGB) {constB=100}, Alpha=ScalarParameter() {parameterName=Flash})
  MaterialFunctionCall(Emissive=Multiply(A=LinearInterpolate(B=@n15, Alpha=@n16.Result) {constA=1, constB=3}, B=Multiply(A=@n18, B=@n9.RGB)) {constA=1}) {materialFunction=&MF_FX_EmissiveAdjustments}
  n19 = Multiply(A=ScalarParameter() {defaultValue=1, parameterName=SpiralSine1}, B=TextureCoordinate())
  n20 = Add(
   A=Multiply(
    A=ComponentMask(None=@n19) {g=true}
    B=MaterialFunctionCall() {materialFunction=/Engine/Functions/Engine_MaterialFunctions02/WorldPositionOffset/ObjectScale}.Scale X
   )
   B=Time()
  )
  n21 = Constant3Vector() {constant=<rgba 0 -1.5 1.5 0>}
  n22 = Constant()
  n23 = Multiply(A=ScalarParameter() {defaultValue=1, parameterName=SpiralSine2}, B=TextureCoordinate())
  n24 = Add(
   A=Multiply(
    A=ComponentMask(None=@n23) {g=true}
    B=MaterialFunctionCall() {materialFunction=&ObjectScale}.Scale X
   )
   B=Time()
  )
  n25 = AppendVector(A=Sine(None=@n24) {period=11}, B=Sine(None=@n24) {period=14.15})
  n26 = Constant3Vector() {constant=<rgba 0 1.5 -1.5 0>}
  n27 = Add(
   A=Multiply(A=Sine(None=@n20) {period=12.9}, B=Transform(None=@n21) {transformSourceType=TRANSFORMSOURCE_Local})
   B=Multiply(A=AppendVector(A=@n22, B=@n25), B=Transform(None=@n26) {transformSourceType=TRANSFORMSOURCE_Local})
  )
  n28 = TextureCoordinate()
  n29 = Add(
   A=Multiply(
    A=ComponentMask(None=@n28) {g=true}
    B=MaterialFunctionCall() {materialFunction=&ObjectScale}.Scale X
   )
   B=Multiply(A=Time(), B=ScalarParameter() {defaultValue=2, parameterName=ReturnWiggleSpeed}) {constB=2}
  )
  n30 = Constant3Vector() {constant=<rgba 0 0 3 0>}
  n31 = Multiply(
   A=Multiply(A=Sine(None=@n29) {period=7}, B=Transform(None=@n30) {transformSourceType=TRANSFORMSOURCE_Local})
   B=ScalarParameter() {parameterName=ReturnWiggle}
  )
  n32 = TextureCoordinate()
  Multiply(
   A=Multiply(A=Add(A=@n27, B=@n31), B=ScalarParameter() {parameterName=WiggleAmount})
   B=SphereMask(A=ComponentMask(None=@n32) {g=true}, B=Constant() {r=0.5}, Radius=ScalarParameter() {defaultValue=0.5, parameterName=SphereRadius}) {attenuationRadius=0.5, hardnessPercent=50}
  )
  Multiply(A=@n13) {constB=2}
  TextureCoordinate()
 ]
}
