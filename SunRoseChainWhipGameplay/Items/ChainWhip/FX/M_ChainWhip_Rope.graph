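# Masked, two-sided material for the Chain Whip's rope, drawing a glowing green chain that can flash and wiggle.
# The chain texture tiles with object X scale, and vertex colors weight its red and green channels for the shape.
# Emissive combines a fade to black along V, time-of-day brightness and a Flash boost.
# Layered sine offsets, including a parameter-driven return wiggle, animate the rope, with a SphereMask centered mid-V tapering motion toward both ends.
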Material /SunRoseChainWhipGameplay/Items/ChainWhip/FX/M_ChainWhip_Rope
graph {
 properties {blendMode=BLEND_Masked, pixelDepthOffsetMode=PDOM_Legacy, twoSided=true, usedShadingModels=Default Lit}
 nodes [
  n1 = LinearInterpolate(Alpha=Divide(A=ObjectRadius(), B=ScalarParameter() {defaultValue=2048, parameterName=MaxLength}))
  Clamp(None=@n1)
  MaterialFunctionCall(Speed=ScalarParameter() {defaultValue=2, parameterName=Param_3}) {materialFunction=&DebugTimeSine}
  n2 = VectorParameter() {defaultValue=<rgba 0.045 1 0.1922 1>}
  n3 = MaterialFunctionCall(Float2=TextureCoordinate()) {materialFunction=&BreakOutFloat2Components}
  n4 = Constant() {r=0.1}
  n5 = MaterialFunctionCall(Float2=TextureCoordinate()) {materialFunction=&BreakOutFloat2Components}
  n6 = ScalarParameter() {defaultValue=8}
  n7 = MaterialFunctionCall() {materialFunction=&ObjectScale}
  n8 = MaterialFunctionCall(R=@n5.R, G=Multiply(A=Multiply(A=@n6, B=@n7.Scale X), B=@n5.G) {constB=20}) {materialFunction=&AppendMany}
  n9 = TextureSample(UVs=MaterialFunctionCall(
    UVs=MaterialFunctionCall(XY=@n8.RG) {materialFunction=&Swizzle}.YX
    Texture Scale=Constant2Vector() {g=1, r=1}
   ) {materialFunction=&ScaleUVsByCenter}.UVs) {texture=&/T_ChainWhip_Chain}
  n10 = VertexColor()
  n11 = Add(A=Multiply(A=@n9.R, B=@n10.R), B=Multiply(A=@n9.G, B=@n10.G))
  n12 = Multiply(A=LinearInterpolate(A=@n2.RGB, B=Constant3Vector() {constant=<rgba 0 0 0 1>}, Alpha=Power(Base=@n3.G, Exp=@n4)), B=@n11)
  n13 = Constant() {r=0.5}
  n14 = Multiply(
   A=Constant() {r=150}
   B=MaterialFunctionCall(Value_Morning=@n13, Value_Day=Constant() {r=1}, Value_Evening=@n13, Value_Night=Constant() {r=0.3}) {materialFunction=&Scalar_Time_of_Day}.Result
  )
  n15 = Multiply(A=ScalarParameter() {parameterName=Flash}, B=Multiply(A=@n2.RGB) {constB=500}) {constB=1000}
  MaterialFunctionCall(Emissive=Add(A=Multiply(A=@n12, B=@n14), B=Multiply(A=@n15, B=@n11))) {materialFunction=&MF_FX_EmissiveAdjustments}
  n16 = Multiply(A=ScalarParameter() {defaultValue=1, parameterName=SpiralSine1}, B=TextureCoordinate())
  n17 = Add(
   A=Multiply(
    A=ComponentMask(None=@n16) {g=true}
    B=MaterialFunctionCall() {materialFunction=/Engine/Functions/Engine_MaterialFunctions02/WorldPositionOffset/ObjectScale}.Scale X
   )
   B=Time()
  )
  n18 = Constant3Vector() {constant=<rgba 0 -1.5 1.5 0>}
  n19 = Constant()
  n20 = Multiply(A=ScalarParameter() {defaultValue=1, parameterName=SpiralSine2}, B=TextureCoordinate())
  n21 = Add(
   A=Multiply(
    A=ComponentMask(None=@n20) {g=true}
    B=MaterialFunctionCall() {materialFunction=&ObjectScale}.Scale X
   )
   B=Time()
  )
  n22 = AppendVector(A=Sine(None=@n21) {period=11}, B=Sine(None=@n21) {period=14.15})
  n23 = Constant3Vector() {constant=<rgba 0 1.5 -1.5 0>}
  n24 = Add(
   A=Multiply(A=Sine(None=@n17) {period=12.9}, B=Transform(None=@n18) {transformSourceType=TRANSFORMSOURCE_Local})
   B=Multiply(A=AppendVector(A=@n19, B=@n22), B=Transform(None=@n23) {transformSourceType=TRANSFORMSOURCE_Local})
  )
  n25 = TextureCoordinate()
  n26 = Add(
   A=Multiply(
    A=ComponentMask(None=@n25) {g=true}
    B=MaterialFunctionCall() {materialFunction=&ObjectScale}.Scale X
   )
   B=Multiply(A=Time(), B=ScalarParameter() {defaultValue=2, parameterName=ReturnWiggleSpeed}) {constB=2}
  )
  n27 = Constant3Vector() {constant=<rgba 0 0 3 0>}
  n28 = Multiply(
   A=Multiply(A=Sine(None=@n26) {period=7}, B=Transform(None=@n27) {transformSourceType=TRANSFORMSOURCE_Local})
   B=ScalarParameter() {parameterName=ReturnWiggle}
  )
  n29 = TextureCoordinate()
  Multiply(
   A=Multiply(A=Add(A=@n24, B=@n28), B=ScalarParameter() {defaultValue=15, parameterName=WiggleAmount})
   B=SphereMask(A=ComponentMask(None=@n29) {g=true}, B=Constant() {r=0.5}, Radius=ScalarParameter() {defaultValue=0.5, parameterName=SphereRadius}) {attenuationRadius=0.5, hardnessPercent=50}
  )
  Saturate(None=Multiply(A=@n11, B=Constant() {r=100}))
  Saturate(None=@n1)
 ]
}
