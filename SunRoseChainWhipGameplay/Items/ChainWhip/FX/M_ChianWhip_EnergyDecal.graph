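# Additive unlit Niagara sprite material for the chain whip's energy decal effect, tinted by particle color.
# A soft radial UV mask combines with particle alpha, a distance-behind-translucency term and depth fade to shape opacity.
# On deferred rendering, a Fresnel from the flattenable scene world normal modulates that mask, so the glow follows underlying surfaces.
# Forward and mobile paths substitute constants for the scene-normal Fresnel.
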Material /SunRoseChainWhipGameplay/Items/ChainWhip/FX/M_ChianWhip_EnergyDecal
graph {
 properties {bUsedWithNiagaraSprites=true, blendMode=BLEND_Additive, pixelDepthOffsetMode=PDOM_Legacy, shadingModel=MSM_Unlit, usedShadingModels=Unlit}
 nodes [
  n1 = ParticleColor()
  n2 = Constant() {r=0.5}
  n3 = Power(Base=SphereMask(A=TextureCoordinate(), B=@n2, Radius=@n2) {hardnessPercent=0})
  n4 = SceneTexture() {sceneTextureId=PPI_WorldNormal}
  n5 = Fresnel(
   ExponentIn=Multiply(A=ScalarParameter() {defaultValue=1, group=Alpha, parameterName=FresnelExponent})
   Normal=MaterialFunctionCall(Normal=ComponentMask(None=@n4.Color) {b=true, g=true, r=true}, Flatness=ScalarParameter() {group=Alpha, parameterName=FlattenNormal}) {materialFunction=&FlattenNormal}.Result
  ) {baseReflectFraction=0, exponent=1}
  n6 = Constant3Vector() {constant=<rgba 0.5 0.5 0.5 1>}
  n7 = ShadingPathSwitch(Default=@n5, Deferred=@n5, Forward=@n6, Mobile=@n6)
  n8 = Multiply(
   A=ShadingPathSwitch(Default=@n7, Deferred=@n7, Mobile=Constant() {r=0.3}) {desc=Migrated from FeatureLevelSwitch}
   B=MaterialFunctionCall!(
    Distance Bias=ScalarParameter() {defaultValue=20, parameterName=CameraOffsetBias}
    Distance Divisor=ScalarParameter() {defaultValue=20, parameterName=DistanceDivisor}
    Scene Depth
    Pixel Depth
   ) {materialFunction=/Game/Effects/Fort_Effects/Materials/_Functions/MF_DistanceBehindTranslucency_01}.Inverted
  )
  DepthFade(Opacity=Multiply(A=@n1.A, B=Multiply(A=@n3, B=@n8) {constA=1})) {fadeDistanceDefault=10}
  LinearInterpolate(B=@n3, Alpha=ScalarParameter() {parameterName=UseSingleSpriteBottomFade}) {constA=1}
  MaterialFunctionCall(Emissive=@n1.RGB) {materialFunction=&MF_FX_EmissiveAdjustments}
  Multiply(A=@n1.RGB)
 ]
}
