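# Niagara system for the Sun Rose Chain Whip's bite effect, a short-lived bite-mark sprite from a single BiteMark emitter.
# User.BiteMarkOn gates a slightly delayed burst through spawn probability, and User.IsLocalPlayer selects the random rotation range.
# The camera-offset sprite shrinks, shifts from overbright yellow to green, and flashes in before fading, using two stacked ScaleColor modules.
# The other user parameters, such as ImpactFXDirection and SwingSectionIndex, are not read by the emitter's shown modules.
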NiagaraSystem /SunRoseChainWhipGameplay/Items/ChainWhip/FX/NS_SunRose_ChainWhip_Bite
dependencies {
 usedDataInterfaces [&NiagaraDataInterfaceCurve, &NiagaraDataInterfaceColorCurve]
 usedDynamicInputs [&FN_LOD_Distance_Remap]
 usedModules [
  &SystemState
  &self:SystemSpawnScript.SystemScriptSource.SystemScriptGraph.NiagaraNodeAssignment_0.SetVariables_#1
 ]
}
emitters {BiteMark {
  emitterUpdateScript [
   &EmitterState()
   &SpawnBurst_Instantaneous(
    Spawn Probability=&MakeCustomFloatFromBool(Bool=$User.BiteMarkOn, True Float=1, False Float=0)
    Spawn Time=0.05
    Use Spawn Probability=true
    UseLoopCountLimit=false
   )
  ]
  particleSpawnScript [
   &InitializeParticle(
    Lifetime=0.55
    Sprite Rotation Angle=&MakeCustomFloatFromBool(
     Bool=$User.IsLocalPlayer
     True Float=&RandomRangeFloat(Minimum=0.3, Maximum=0.7)
     False Float=&RandomRangeFloat(Minimum=0.15, Maximum=-0.07)
    )
    Sprite Rotation Mode=::NewEnumerator1
    Sprite Size=<80 42>
    Sprite Size Mode=::NewEnumerator4
    UsePositionOffset=false
    Write Material Random=true
   )
   &CameraOffset(Camera Offset Amount=20)
  ]
  particleUpdateScript [
   &ParticleState()
   &ScaleSpriteSize(Uniform Curve Sprite Scale=DataInterface() {propertyValues {Curve {keys [
        {arriveTangent=-2.8734, interpMode=RCIM_Cubic, leaveTangent=-2.8734, tangentMode=RCTM_User, time=0.1, value=1}
        {arriveTangent=-0.3333, interpMode=RCIM_Cubic, leaveTangent=-0.3333, time=0.3, value=0.75}
        {interpMode=RCIM_Cubic, tangentMode=RCTM_User, time=1, value=0.7}
       ]}}})
   &ScaleColor(
    Linear Color Curve=DataInterface() {propertyValues {
      AlphaCurve {keys [<at 0 1>,<at 1 0>]}
      BlueCurve {keys [<at 0 0.4123>,<at 0.2834 0.394>]}
      GreenCurve {keys [<at 0 2.2475>,<at 0.2834 1.2>]}
      RedCurve {keys [<at 0 3>,<at 0.2834 0>]}
     }}
    Scale Mode=::NewEnumerator2
   ) {module=ScaleColor001}
   &ScaleColor(
    Scale Alpha=&FloatFromCurve(FloatCurve=DataInterface() {propertyValues {Curve {keys [
         {arriveTangent=9.0725, interpMode=RCIM_Cubic, leaveTangent=9.0725, tangentMode=RCTM_User, time=0, value=0}
         {arriveTangent=-2.5249, interpMode=RCIM_Cubic, leaveTangent=-2.5249, tangentMode=RCTM_User, time=0.1408, value=0.9847}
         {arriveTangent=-0.0405, interpMode=RCIM_Cubic, leaveTangent=-0.0405, tangentMode=RCTM_User, time=1, value=0}
        ]}}})
    Scale RGB=&VectorFromFloat(Value=&FloatFromCurve(
      FloatCurve=DataInterface() {propertyValues {Curve {keys [{interpMode=RCIM_Cubic, tangentMode=RCTM_User, time=0, value=1}, {interpMode=RCIM_Cubic, tangentMode=RCTM_User, time=1, value=0}]}}}
      Scale Curve=2
     ))
    ScaleA=true
    ScaleRGB=true
   )
  ]
  rendererClasses [&NiagaraSpriteRendererProperties]
  renderers [{rendererClass=&NiagaraSpriteRendererProperties}]
 }}
summary {
 systemName=NS_SunRose_ChainWhip_Bite
 emitters [{emitterName=BiteMark, rendererClasses [&NiagaraSpriteRendererProperties]}]
 userVariables [
  User.bIsFrontEnd {type=NiagaraBool}
  User.bIsFrontEndPreview {type=NiagaraBool}
  User.BiteMarkOn {type=NiagaraBool, defaultValue=true}
  User.ImpactFXDirection {type=Vector3f, description="", defaultValue {struct=Vector3f, value=<0 0 0>}}
  User.IsLocalPlayer {type=NiagaraBool}
  User.SwingSectionIndex {type=NiagaraInt32, defaultValue=-1}
 ]
}
