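; Sound library context for the Chain Whip item that controls how its library sounds are played.
; Its Play function breaks the event input into a sound and owner, then casts the owner to FortWeapon.
; It spawns the sound attached to that weapon's mesh and returns the spawned audio component in an array with a true result.
; The failure branch is empty, so no sound spawns there.
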(fn Play
  (params (InEventData :type Sound%20Library%20Context%20Event%20Input%20Structure%20%28by%20ref%29))
  (bind (_sound _owner) (Utilities|Struct|BreakSoundLibraryContextEventInput InEventData))
  (bind _asfort_weapon (Utilities|Casting|CastToFortWeapon _owner))
  (bind _weaponmesh (Variables|/Script/FortniteGame.FortWeapon|GetWeaponMesh _asfort_weapon)
    (:then
      (bind _returnvalue (Audio|SpawnSoundAttached _sound _weaponmesh))
      (return (Utilities|Array|MakeArray _returnvalue) true))
    (:else)))
