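# Looping ambient sound cue for the ChainWhip weapon, reusing a Ghost Pistol trail loop wave at low volume.
# A looping wave player feeds a modulator, which all quality-level branches share, so every quality setting plays the same sound.
# It routes to the weapon foley sound class with a ChainWhip ambient attenuation asset and a melee weapon concurrency set.
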SoundCue /SunRoseChainWhipGameplay/Items/ChainWhip/Sound/ChainWhip_Ambient_Loop_Cue
root=QualityLevel_0
nodes [
 {id=Modulator_1, children [WavePlayer_7]}
 {id=QualityLevel_0, children [Modulator_1,Modulator_1,Modulator_1,Modulator_1]}
 {id=WavePlayer_7, properties {bLooping=true, soundWaveAssetPtr=/Game/Sounds/Fort_Weapons/pistol/Ghost/SFX_GhostPistol_GhostTrail_Loop_04}}
]
properties {
 attenuationSettings=&/ChainWhip_Ambient_Att
 maxDistance=2600
 soundClassObject=&Fort_WeaponFoley
 volumeMultiplier=0.2
 attenuationOverrides {distanceAlgorithm=NaturalSound, falloffDistance=3000, attenuationShapeExtents {x=1000}}
 concurrencySet [&Player_Melee_Weapon_Conc]
}
