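# Sound cue for the chain whip landing after being dropped, routed to the loot collect sound class with weapon drop attenuation.
# A mixer layers a boosted drop-land wave, passed through a modulator with fixed pitch, over a much quieter random pick of four borrowed Embers pickaxe impacts.
# Overridden concurrency caps simultaneous plays at a few and blocks new ones beyond that.
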SoundCue /SunRoseChainWhipGameplay/Items/ChainWhip/Sound/ChainWhip_DropLanded_Cue
root=Mixer_0
nodes [
 {id=Mixer_0, children [Modulator_1,Random_0], properties {inputVolume [3,0.2]}}
 {id=Modulator_1, children [WavePlayer_9], properties {pitchMax=1, pitchMin=1}}
 {id=Random_0, children [WavePlayer_1,WavePlayer_2,WavePlayer_3,WavePlayer_4], properties {weights [1,1,1,1]}}
 {id=WavePlayer_1, properties {soundWaveAssetPtr=/BRCosmetics/Sounds/PickAxes/Embers/PA_Embers__Impact_01}}
 {id=WavePlayer_2, properties {soundWaveAssetPtr=/BRCosmetics/Sounds/PickAxes/Embers/PA_Embers__Impact_02}}
 {id=WavePlayer_3, properties {soundWaveAssetPtr=/BRCosmetics/Sounds/PickAxes/Embers/PA_Embers__Impact_03}}
 {id=WavePlayer_4, properties {soundWaveAssetPtr=/BRCosmetics/Sounds/PickAxes/Embers/PA_Embers__Impact_04}}
 {id=WavePlayer_9, properties {soundWaveAssetPtr=&/ChainWhip_DropLand}}
]
properties {
 attenuationSettings=&Fort_Weapon_Drop_Att
 bOverrideConcurrency=true
 maxDistance=3000
 soundClassObject=&Fort_Loot_Collect
 volumeMultiplier=0.5
 attenuationOverrides {falloffDistance=1000, attenuationShapeExtents {x=0}}
 concurrencyOverrides {maxCount=3, resolutionRule=PreventNew}
}
