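# Sound cue that plays when the Chain Whip item is dropped.
# A random node picks, with equal weight, one of five Chainmail pickaxe swing waves borrowed from BRCosmetics, then passes it through a modulator.
# It plays at reduced volume through weapon-drop attenuation and the loot-collect sound class, with a shared stop-oldest weapon-drop concurrency set.
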SoundCue /SunRoseChainWhipGameplay/Items/ChainWhip/Sound/ChainWhip_Drop_Cue
root=Modulator_1
nodes [
 {id=Modulator_1, children [Random_1]}
 {id=Random_1, children [WavePlayer_0,WavePlayer_1,WavePlayer_5,WavePlayer_6,WavePlayer_7], properties {weights [1,1,1,1,1]}}
 {id=WavePlayer_0, properties {soundWaveAssetPtr=/BRCosmetics/Sounds/PickAxes/Chainmail/Pickaxe_Chainmail_Swing_01}}
 {id=WavePlayer_1, properties {soundWaveAssetPtr=/BRCosmetics/Sounds/PickAxes/Chainmail/Pickaxe_Chainmail_Swing_02}}
 {id=WavePlayer_5, properties {soundWaveAssetPtr=/BRCosmetics/Sounds/PickAxes/Chainmail/Pickaxe_Chainmail_Swing_03}}
 {id=WavePlayer_6, properties {soundWaveAssetPtr=/BRCosmetics/Sounds/PickAxes/Chainmail/Pickaxe_Chainmail_Swing_04}}
 {id=WavePlayer_7, properties {soundWaveAssetPtr=/BRCosmetics/Sounds/PickAxes/Chainmail/Pickaxe_Chainmail_Swing_05}}
]
properties {
 attenuationSettings=&Fort_Weapon_Drop_Att
 maxDistance=3000
 soundClassObject=&Fort_Loot_Collect
 volumeMultiplier=0.4
 concurrencyOverrides {resolutionRule=PreventNew}
 concurrencySet [&Weapon_Drop_StopOldest]
}
