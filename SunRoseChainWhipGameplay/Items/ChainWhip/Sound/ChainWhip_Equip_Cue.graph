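# Sound cue played for equipping the chain whip, built from reused whip swing and hook-shot catch sounds.
# A mixer layers one of five random swing whooshes with one of three random hook-shot catch sounds, delayed briefly and mixed at half volume.
# Each layer passes through a modulator for variation.
# It uses weapon reload attenuation, the weapon foley sound class and its own concurrency limit.
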SoundCue /SunRoseChainWhipGameplay/Items/ChainWhip/Sound/ChainWhip_Equip_Cue
root=Mixer_2
nodes [
 {id=Delay_0, children [Modulator_0], properties {delayMax=0.2, delayMin=0.2}}
 {id=Mixer_2, children [Modulator_2,Delay_0], properties {inputVolume [1,0.5]}}
 {id=Modulator_0, children [Random_0]}
 {id=Modulator_2, children [Random_3]}
 {id=Random_0, children [WavePlayer_0,WavePlayer_1,WavePlayer_2], properties {weights [1,1,1]}}
 {
  id=Random_3
  children [WavePlayer_14, WavePlayer_15, WavePlayer_16, WavePlayer_17, WavePlayer_18]
  properties {bShouldExcludeFromBranchCulling=true, weights [1,1,1,1,1]}
 }
 {id=WavePlayer_0, properties {soundWaveAssetPtr=&/ChainWhip_HookShot_Catch_03}}
 {id=WavePlayer_1, properties {soundWaveAssetPtr=&/ChainWhip_HookShot_Catch_02}}
 {id=WavePlayer_14, properties {soundWaveAssetPtr=&/ChainWhip_Swing_Whoosh_01}}
 {id=WavePlayer_15, properties {soundWaveAssetPtr=&/ChainWhip_Swing_Whoosh_02}}
 {id=WavePlayer_16, properties {soundWaveAssetPtr=&/ChainWhip_Swing_Whoosh_03}}
 {id=WavePlayer_17, properties {soundWaveAssetPtr=&/ChainWhip_Swing_Whoosh_04}}
 {id=WavePlayer_18, properties {soundWaveAssetPtr=&/ChainWhip_Swing_Whoosh_05}}
 {id=WavePlayer_2, properties {soundWaveAssetPtr=&/ChainWhip_HookShot_Catch_01}}
]
properties {
 attenuationSettings=&Fort_Weapon_Reload_Att
 bOverrideConcurrency=true
 maxDistance=2250
 soundClassObject=&Fort_WeaponFoley
 volumeMultiplier=0.351
 concurrencyOverrides {maxCount=3}
}
