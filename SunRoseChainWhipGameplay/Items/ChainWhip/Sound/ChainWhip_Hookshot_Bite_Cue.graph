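# Sound cue for the chain whip hookshot bite, playing one of three equally weighted random catch sounds.
# An istarget branch sends the target a quieter version using the third-person fire sound class and the hookshot 2D attenuation.
# Otherwise an isinstigator branch picks a boosted first-person class for the instigator or an unboosted third-person class for others.
# Both of those paths use close-range melee stereo attenuation with a stereo spread override.
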SoundCue /SunRoseChainWhipGameplay/Items/ChainWhip/Sound/ChainWhip_Hookshot_Bite_Cue
root=Branch_0
nodes [
 {
  id=Attenuation_1
  children [Random_0]
  properties {attenuationSettings=&/ChainWhip_HookShot_Fire_2D_Att}
 }
 {
  id=Attenuation_3
  children [Random_2]
  properties {attenuationSettings=/Game/Sounds/Attenuation/Fort_Weapon_Melee_Attack_Close_Range_Stereo_Att, attenuationOverrides {stereoSpread=29871.3984}}
 }
 {
  id=Attenuation_4
  children [Random_1]
  properties {attenuationSettings=&Fort_Weapon_Melee_Attack_Close_Range_Stereo_Att, attenuationOverrides {stereoSpread=29871.3984}}
 }
 {id=Branch_0, children [Modulator_3,Branch_3,Branch_3], properties {boolParameterName=istarget}}
 {id=Branch_3, children [Modulator_1,Modulator_2,Modulator_2], properties {boolParameterName=isinstigator}}
 {id=Modulator_1, children [SoundClass_2], properties {volumeMax=1.1, volumeMin=1.1}}
 {id=Modulator_2, children [SoundClass_3]}
 {id=Modulator_3, children [SoundClass_1], properties {volumeMax=0.6, volumeMin=0.6}}
 {id=Random_0, children [WavePlayer_0,WavePlayer_1,WavePlayer_2], properties {weights [1,1,1]}}
 {id=Random_1, children [WavePlayer_6,WavePlayer_7,WavePlayer_8], properties {weights [1,1,1]}}
 {id=Random_2, children [WavePlayer_9, WavePlayer_10, WavePlayer_11], properties {weights [1,1,1]}}
 {
  id=SoundClass_1
  children [Attenuation_1]
  properties {soundClassOverride=&Fort_WeaponRanged_Fire_3P_PSM}
 }
 {
  id=SoundClass_2
  children [Attenuation_4]
  properties {soundClassOverride=&Fort_WeaponRanged_Fire_1P_PSM}
 }
 {
  id=SoundClass_3
  children [Attenuation_3]
  properties {soundClassOverride=&Fort_WeaponRanged_Fire_3P_PSM}
 }
 {id=WavePlayer_0, properties {soundWaveAssetPtr=&/ChainWhip_HookShot_Catch_01}}
 {id=WavePlayer_1, properties {soundWaveAssetPtr=&/ChainWhip_HookShot_Catch_02}}
 {id=WavePlayer_10, properties {soundWaveAssetPtr=&ChainWhip_HookShot_Catch_02}}
 {id=WavePlayer_11, properties {soundWaveAssetPtr=&/ChainWhip_HookShot_Catch_03}}
 {id=WavePlayer_2, properties {soundWaveAssetPtr=&ChainWhip_HookShot_Catch_03}}
 {id=WavePlayer_6, properties {soundWaveAssetPtr=&ChainWhip_HookShot_Catch_01}}
 {id=WavePlayer_7, properties {soundWaveAssetPtr=&ChainWhip_HookShot_Catch_02}}
 {id=WavePlayer_8, properties {soundWaveAssetPtr=&ChainWhip_HookShot_Catch_03}}
 {id=WavePlayer_9, properties {soundWaveAssetPtr=&ChainWhip_HookShot_Catch_01}}
]
properties {maxDistance=10058.75, soundClassObject=&Fort_WeaponMelee_PSM, volumeMultiplier=0.5}
