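# Sound cue for the Chain Whip hookshot exit, playing one of three randomly chosen ghost pistol trail waves at slightly lowered pitch.
# An enveloper fades the wave out, then a quality node with identical inputs feeds volume modulators.
# Branches on islocal and isenemy halve the volume for non-local, non-enemy sources and keep it full otherwise.
# It uses Chain Whip attenuation, targeting concurrency and a melee weapon sound class.
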SoundCue /SunRoseChainWhipGameplay/Items/ChainWhip/Sound/ChainWhip_Hookshot_Exit_Cue
root=Branch_2
nodes [
 {id=Branch_2, children [Modulator_2,Branch_3,Modulator_2], properties {boolParameterName=islocal}}
 {id=Branch_3, children [Modulator_4,Modulator_3,Modulator_4], properties {boolParameterName=isenemy}}
 {
  id=Enveloper_1
  children [Random_1]
  properties {pitchCurve {editorCurveData {keys [{time=0}]}}, volumeCurve {editorCurveData {keys [<at 0 1>,<at 0.5 0.75>,<at 0.809 0.2639>,<at 1.5 0>]}}}
 }
 {id=Modulator_2, children [QualityLevel_0], properties {pitchMax=1, pitchMin=1, volumeMax=1, volumeMin=1}}
 {id=Modulator_3, children [QualityLevel_0], properties {pitchMax=1, pitchMin=1, volumeMax=0.5, volumeMin=0.5}}
 {id=Modulator_4, children [QualityLevel_0], properties {pitchMax=1, pitchMin=1, volumeMax=1, volumeMin=1}}
 {id=QualityLevel_0, children [Enveloper_1,Enveloper_1,Enveloper_1,Enveloper_1]}
 {id=Random_1, children [WavePlayer_3,WavePlayer_4,WavePlayer_5], properties {weights [1,1,1]}}
 {id=WavePlayer_3, properties {soundWaveAssetPtr=/Game/Sounds/Fort_Weapons/pistol/Ghost/SFX_GhostPistol_GhostTrail_01}}
 {id=WavePlayer_4, properties {soundWaveAssetPtr=/Game/Sounds/Fort_Weapons/pistol/Ghost/SFX_GhostPistol_GhostTrail_02}}
 {id=WavePlayer_5, properties {soundWaveAssetPtr=/Game/Sounds/Fort_Weapons/pistol/Ghost/SFX_GhostPistol_GhostTrail_03}}
]
properties {
 attenuationSettings=&/ChainWhip_Att
 maxDistance=14000
 pitchMultiplier=0.9
 soundClassObject=&Fort_WeaponMelee_PSM
 attenuationOverrides {distanceAlgorithm=NaturalSound, falloffDistance=3000, attenuationShapeExtents {x=1000}}
 concurrencySet [&/ChainWhip_Targetting_Conc]
}
