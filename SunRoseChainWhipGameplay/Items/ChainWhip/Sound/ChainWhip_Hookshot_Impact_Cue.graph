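# Sound cue for the Chain Whip hookshot impact, with concurrency capped at three.
# A mixer layers randomized Embers pickaxe impacts, modulated and using close attenuation, with randomized Molotov firefly explosions using distant mono attenuation.
# Each layer overrides its sound class for third-person or distant ranged fire.
# Branches on islocal and isenemy halve the mix volume for non-local, non-enemy sources.
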SoundCue /SunRoseChainWhipGameplay/Items/ChainWhip/Sound/ChainWhip_Hookshot_Impact_Cue
root=Branch_1
nodes [
 {
  id=Attenuation_118
  children [Modulator_0]
  properties {attenuationSettings=&Fort_Weapon_Projectile_Close_Athena_Att, attenuationOverrides {stereoSpread=29871.3984}}
 }
 {
  id=Attenuation_120
  children [Random_3]
  properties {attenuationSettings=&Fort_Weapon_Projectile_Distant_Mono_Athena_Att}
 }
 {id=Branch_1, children [Modulator_2,Branch_2,Modulator_2], properties {boolParameterName=islocal}}
 {id=Branch_2, children [Modulator_4,Modulator_3,Modulator_4], properties {boolParameterName=isenemy}}
 {id=Mixer_1, children [SoundClass_77, SoundClass_76], properties {inputVolume [1,1]}}
 {id=Modulator_0, children [Random_1]}
 {id=Modulator_2, children [Mixer_1], properties {pitchMax=1, pitchMin=1, volumeMax=1, volumeMin=1}}
 {id=Modulator_3, children [Mixer_1], properties {pitchMax=1, pitchMin=1, volumeMax=0.5, volumeMin=0.5}}
 {id=Modulator_4, children [Mixer_1], properties {pitchMax=1, pitchMin=1, volumeMax=1, volumeMin=1}}
 {id=Random_1, children [WavePlayer_10, WavePlayer_11, WavePlayer_12, WavePlayer_18], properties {weights [1,1,1,1]}}
 {id=Random_3, children [WavePlayer_19, WavePlayer_20, WavePlayer_21, WavePlayer_22, WavePlayer_23], properties {weights [1,1,1,1,1]}}
 {
  id=SoundClass_76
  children [Attenuation_120]
  properties {soundClassOverride=&Fort_WeaponRanged_Fire_Distant}
 }
 {
  id=SoundClass_77
  children [Attenuation_118]
  properties {soundClassOverride=&Fort_WeaponRanged_Fire_3P_PSM}
 }
 {id=WavePlayer_10, properties {soundWaveAssetPtr=/BRCosmetics/Sounds/PickAxes/Embers/PA_Embers__Impact_01}}
 {id=WavePlayer_11, properties {soundWaveAssetPtr=/BRCosmetics/Sounds/PickAxes/Embers/PA_Embers__Impact_02}}
 {id=WavePlayer_12, properties {soundWaveAssetPtr=/BRCosmetics/Sounds/PickAxes/Embers/PA_Embers__Impact_03}}
 {id=WavePlayer_18, properties {soundWaveAssetPtr=/BRCosmetics/Sounds/PickAxes/Embers/PA_Embers__Impact_04}}
 {id=WavePlayer_19, properties {soundWaveAssetPtr=/Game/Sounds/Fort_Weapons/Grenades/FireJar/Molotov_FireflyExplosion_01}}
 {id=WavePlayer_20, properties {soundWaveAssetPtr=/Game/Sounds/Fort_Weapons/Grenades/FireJar/Molotov_FireflyExplosion_02}}
 {id=WavePlayer_21, properties {soundWaveAssetPtr=/Game/Sounds/Fort_Weapons/Grenades/FireJar/Molotov_FireflyExplosion_03}}
 {id=WavePlayer_22, properties {soundWaveAssetPtr=/Game/Sounds/Fort_Weapons/Grenades/FireJar/Molotov_FireflyExplosion_04}}
 {id=WavePlayer_23, properties {soundWaveAssetPtr=/Game/Sounds/Fort_Weapons/Grenades/FireJar/Molotov_FireflyExplosion_05}}
]
properties {
 bOverrideConcurrency=true
 maxDistance=27500
 soundClassObject=&Fort_WeaponMelee_PSM
 volumeMultiplier=0.5
 concurrencyOverrides {maxCount=3}
}
