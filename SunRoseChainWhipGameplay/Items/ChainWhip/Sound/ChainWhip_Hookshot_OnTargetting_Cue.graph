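# Sound cue for the Chain Whip hookshot targeting event, playing one of two soul sweetener swing samples at random with slight pitch variation.
# Branches on islocal and isenemy halve the volume when both are false, leaving local, enemy and unset cases at full level.
# All quality levels share the same random pair.
# It uses Chain Whip attenuation, targeting concurrency and the melee weapon sound class, with a reduced overall volume.
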SoundCue /SunRoseChainWhipGameplay/Items/ChainWhip/Sound/ChainWhip_Hookshot_OnTargetting_Cue
root=Branch_2
nodes [
 {id=Branch_2, children [Modulator_2,Branch_3,Modulator_2], properties {boolParameterName=islocal}}
 {id=Branch_3, children [Modulator_4,Modulator_3,Modulator_4], properties {boolParameterName=isenemy}}
 {id=Modulator_1, children [QualityLevel_1], properties {pitchMax=1.1, pitchMin=0.9}}
 {id=Modulator_2, children [Modulator_1], properties {pitchMax=1, pitchMin=1, volumeMax=1, volumeMin=1}}
 {id=Modulator_3, children [Modulator_1], properties {pitchMax=1, pitchMin=1, volumeMax=0.5, volumeMin=0.5}}
 {id=Modulator_4, children [Modulator_1], properties {pitchMax=1, pitchMin=1, volumeMax=1, volumeMin=1}}
 {id=QualityLevel_1, children [Random_2,Random_2,Random_2,Random_2]}
 {id=Random_2, children [WavePlayer_3,WavePlayer_4], properties {weights [1,1]}}
 {id=WavePlayer_3, properties {soundWaveAssetPtr=&/ChainWhip_Swing_Soul_Sweetener_07}}
 {id=WavePlayer_4, properties {soundWaveAssetPtr=&/ChainWhip_Swing_Soul_Sweetener_06}}
]
properties {
 attenuationSettings=&/ChainWhip_Att
 maxDistance=14000
 soundClassObject=&Fort_WeaponMelee_PSM
 volumeMultiplier=0.25
 attenuationOverrides {distanceAlgorithm=NaturalSound, falloffDistance=3000, attenuationShapeExtents {x=1000}}
 concurrencySet [&/ChainWhip_Targetting_Conc]
}
