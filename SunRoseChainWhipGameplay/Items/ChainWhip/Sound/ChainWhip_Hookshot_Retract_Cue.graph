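# Sound cue for the Chain Whip hookshot retract, layering whoosh, sweetener and catch sounds.
# Randomized swing whooshes and projectile-cast sweeteners mix with pitched-up soul sweeteners and a delayed, quieter random catch sound.
# Nested islocal and isenemy branches share that mix, halving volume only for remote non-enemy sources.
# It uses Chain Whip attenuation, a melee weapon sound class and a concurrency override.
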SoundCue /SunRoseChainWhipGameplay/Items/ChainWhip/Sound/ChainWhip_Hookshot_Retract_Cue
root=Branch_0
nodes [
 {id=Branch_0, children [Modulator_2,Branch_3,Modulator_2], properties {boolParameterName=islocal}}
 {id=Branch_3, children [Modulator_4,Modulator_3,Modulator_4], properties {boolParameterName=isenemy}}
 {id=Delay_0, children [Modulator_1], properties {delayMax=0.35, delayMin=0.35}}
 {id=Mixer_1, children [Mixer_2,Modulator_0,Delay_0], properties {inputVolume [1,1,0.8]}}
 {id=Mixer_2, children [Random_1,Random_3], properties {inputVolume [1,1]}}
 {id=Modulator_0, children [Random_2], properties {pitchMax=1.2, pitchMin=1.2}}
 {id=Modulator_1, children [Random_0], properties {pitchMax=1.1, volumeMax=0.8, volumeMin=0.8}}
 {id=Modulator_2, children [Mixer_1], properties {pitchMax=1, pitchMin=1, volumeMax=1, volumeMin=1}}
 {id=Modulator_3, children [Mixer_1], properties {pitchMax=1, pitchMin=1, volumeMax=0.5, volumeMin=0.5}}
 {id=Modulator_4, children [Mixer_1], properties {pitchMax=1, pitchMin=1, volumeMax=1, volumeMin=1}}
 {id=Random_0, children [WavePlayer_17, WavePlayer_18], properties {weights [1,1]}}
 {id=Random_1, children [WavePlayer_3,WavePlayer_4,WavePlayer_5,WavePlayer_6,WavePlayer_7], properties {weights [1,1,1,1,1]}}
 {
  id=Random_2
  children [WavePlayer_8, WavePlayer_10, WavePlayer_11, WavePlayer_12, WavePlayer_19, WavePlayer_20, WavePlayer_21, WavePlayer_22, WavePlayer_26]
  properties {weights [1,1,1,1,1,1,1,1,1]}
 }
 {id=Random_3, children [WavePlayer_0,WavePlayer_1,WavePlayer_2,WavePlayer_9], properties {weights [1,1,1,1]}}
 {id=WavePlayer_0, properties {soundWaveAssetPtr=&/ChainWhip_Prj_Cast_Sweetener_01}}
 {id=WavePlayer_1, properties {soundWaveAssetPtr=&/ChainWhip_Prj_Cast_Sweetener_02}}
 {id=WavePlayer_10, properties {soundWaveAssetPtr=&/ChainWhip_Swing_Soul_Sweetener_08}}
 {id=WavePlayer_11, properties {soundWaveAssetPtr=&/ChainWhip_Swing_Soul_Sweetener_07}}
 {id=WavePlayer_12, properties {soundWaveAssetPtr=&/ChainWhip_Swing_Soul_Sweetener_06}}
 {id=WavePlayer_17, properties {soundWaveAssetPtr=&/ChainWhip_HookShot_Catch_02}}
 {id=WavePlayer_18, properties {soundWaveAssetPtr=&/ChainWhip_HookShot_Catch_03}}
 {id=WavePlayer_19, properties {soundWaveAssetPtr=&/ChainWhip_Swing_Soul_Sweetener_05}}
 {id=WavePlayer_2, properties {soundWaveAssetPtr=&/ChainWhip_Prj_Cast_Sweetener_03}}
 {id=WavePlayer_20, properties {soundWaveAssetPtr=&/ChainWhip_Swing_Soul_Sweetener_04}}
 {id=WavePlayer_21, properties {soundWaveAssetPtr=&/ChainWhip_Swing_Soul_Sweetener_03}}
 {id=WavePlayer_22, properties {soundWaveAssetPtr=&/ChainWhip_Swing_Soul_Sweetener_02}}
 {id=WavePlayer_26, properties {soundWaveAssetPtr=&/ChainWhip_Swing_Soul_Sweetener_01}}
 {id=WavePlayer_3, properties {soundWaveAssetPtr=&/ChainWhip_Swing_Whoosh_01}}
 {id=WavePlayer_4, properties {soundWaveAssetPtr=&/ChainWhip_Swing_Whoosh_02}}
 {id=WavePlayer_5, properties {soundWaveAssetPtr=&/ChainWhip_Swing_Whoosh_03}}
 {id=WavePlayer_6, properties {soundWaveAssetPtr=&/ChainWhip_Swing_Whoosh_04}}
 {id=WavePlayer_7, properties {soundWaveAssetPtr=&/ChainWhip_Swing_Whoosh_05}}
 {id=WavePlayer_8, properties {soundWaveAssetPtr=&/ChainWhip_Swing_Soul_Sweetener_09}}
 {id=WavePlayer_9, properties {soundWaveAssetPtr=&/ChainWhip_Prj_Cast_Sweetener_04}}
]
properties {
 attenuationSettings=&/ChainWhip_Att
 bOverrideConcurrency=true
 maxDistance=14000
 soundClassObject=&Fort_WeaponMelee_PSM
 volumeMultiplier=0.7
 concurrencyOverrides {maxCount=4}
}
