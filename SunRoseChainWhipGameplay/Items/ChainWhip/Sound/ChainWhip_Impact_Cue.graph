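# Sound cue for Chain Whip melee impacts, randomly playing one of four equally weighted Embers pickaxe impact waves through a shared modulator.
# Branches on islocal and isenemy play non-local, non-enemy (teammate) hits at half volume, while local, enemy and unset cases stay at full volume.
# Close-range melee attenuation, a melee weapon sound class and a concurrency override limit how far and how many impacts play.
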SoundCue /SunRoseChainWhipGameplay/Items/ChainWhip/Sound/ChainWhip_Impact_Cue
root=Branch_0
nodes [
 {id=Branch_0, children [Modulator_2,Branch_3,Modulator_2], properties {boolParameterName=islocal}}
 {id=Branch_3, children [Modulator_4,Modulator_3,Modulator_4], properties {boolParameterName=isenemy}}
 {id=Modulator_0, children [Random_0]}
 {id=Modulator_2, children [Modulator_0], properties {pitchMax=1, pitchMin=1, volumeMax=1, volumeMin=1}}
 {id=Modulator_3, children [Modulator_0], properties {pitchMax=1, pitchMin=1, volumeMax=0.5, volumeMin=0.5}}
 {id=Modulator_4, children [Modulator_0], properties {pitchMax=1, pitchMin=1, volumeMax=1, volumeMin=1}}
 {id=Random_0, children [WavePlayer_0,WavePlayer_2,WavePlayer_3,WavePlayer_5], properties {weights [1,1,1,1]}}
 {id=WavePlayer_0, properties {soundWaveAssetPtr=/BRCosmetics/Sounds/PickAxes/Embers/PA_Embers__Impact_01}}
 {id=WavePlayer_2, properties {soundWaveAssetPtr=/BRCosmetics/Sounds/PickAxes/Embers/PA_Embers__Impact_02}}
 {id=WavePlayer_3, properties {soundWaveAssetPtr=/BRCosmetics/Sounds/PickAxes/Embers/PA_Embers__Impact_03}}
 {id=WavePlayer_5, properties {soundWaveAssetPtr=/BRCosmetics/Sounds/PickAxes/Embers/PA_Embers__Impact_04}}
]
properties {
 attenuationSettings=/Game/Sounds/Attenuation/Fort_Weapon_Melee_Attack_Close_Range_Stereo_Att
 bOverrideConcurrency=true
 maxDistance=10058.75
 soundClassObject=&Fort_WeaponMelee_PSM
 volumeMultiplier=0.9
 concurrencyOverrides {maxCount=3}
}
