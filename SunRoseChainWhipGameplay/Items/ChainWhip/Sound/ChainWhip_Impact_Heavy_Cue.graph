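# Sound cue for the Chain Whip heavy impact, varying playback by the listener's role in the hit.
# Branches on istarget, then isinstigator: targets hear quieter random 2D impacts, while instigators and others share a four-wave 3P set.
# The instigator path is pitched up, louder and classed 1P; the bystander path is pitched down and classed 3P.
# Each path has its own attenuation node, and a concurrency override caps simultaneous instances.
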SoundCue /SunRoseChainWhipGameplay/Items/ChainWhip/Sound/ChainWhip_Impact_Heavy_Cue
root=Branch_1
nodes [
 {
  id=Attenuation_0
  children [Random_0]
  properties {attenuationSettings=&/ChainWhip_HookShot_Fire_2D_Att}
 }
 {
  id=Attenuation_1
  children [Random_2]
  properties {attenuationSettings=/Game/Sounds/Attenuation/Fort_Weapon_Melee_Attack_Close_Range_Stereo_Att, attenuationOverrides {stereoSpread=29871.3984}}
 }
 {
  id=Attenuation_2
  children [Random_1]
  properties {attenuationSettings=&Fort_Weapon_Melee_Attack_Close_Range_Stereo_Att, attenuationOverrides {stereoSpread=29871.3984}}
 }
 {id=Branch_1, children [Modulator_2,Branch_2,Branch_2], properties {boolParameterName=istarget}}
 {id=Branch_2, children [Modulator_0,Modulator_1,Modulator_1], properties {boolParameterName=isinstigator}}
 {id=Modulator_0, children [SoundClass_0], properties {pitchMax=1.1, pitchMin=1.1, volumeMax=1.1, volumeMin=1.1}}
 {id=Modulator_1, children [SoundClass_1], properties {pitchMax=0.9, pitchMin=0.9}}
 {id=Modulator_2, children [SoundClass_77], properties {pitchMax=1, pitchMin=1, volumeMax=0.6, volumeMin=0.6}}
 {id=Random_0, children [WavePlayer_0,WavePlayer_1,WavePlayer_2,WavePlayer_3], properties {weights [1,1,1,1]}}
 {id=Random_1, children [WavePlayer_4,WavePlayer_5,WavePlayer_6,WavePlayer_7], properties {weights [1,1,1,1]}}
 {id=Random_2, children [WavePlayer_8, WavePlayer_9, WavePlayer_10, WavePlayer_11], properties {weights [1,1,1,1]}}
 {
  id=SoundClass_0
  children [Attenuation_2]
  properties {soundClassOverride=&Fort_WeaponRanged_Fire_1P_PSM}
 }
 {
  id=SoundClass_1
  children [Attenuation_1]
  properties {soundClassOverride=&Fort_WeaponRanged_Fire_3P_PSM}
 }
 {
  id=SoundClass_77
  children [Attenuation_0]
  properties {soundClassOverride=&Fort_WeaponRanged_Fire_3P_PSM}
 }
 {id=WavePlayer_0, properties {soundWaveAssetPtr=&/ChainWhip_HeavyImpact_2D_01}}
 {id=WavePlayer_1, properties {soundWaveAssetPtr=&/ChainWhip_HeavyImpact_2D_02}}
 {id=WavePlayer_10, properties {soundWaveAssetPtr=&/ChainWhip_HeavyImpact_3P_03}}
 {id=WavePlayer_11, properties {soundWaveAssetPtr=&/ChainWhip_HeavyImpact_3P_04}}
 {id=WavePlayer_2, properties {soundWaveAssetPtr=&/ChainWhip_HeavyImpact_2D_03}}
 {id=WavePlayer_3, properties {soundWaveAssetPtr=&/ChainWhip_HeavyImpact_2D_04}}
 {id=WavePlayer_4, properties {soundWaveAssetPtr=&/ChainWhip_HeavyImpact_3P_01}}
 {id=WavePlayer_5, properties {soundWaveAssetPtr=&/ChainWhip_HeavyImpact_3P_02}}
 {id=WavePlayer_6, properties {soundWaveAssetPtr=&ChainWhip_HeavyImpact_3P_03}}
 {id=WavePlayer_7, properties {soundWaveAssetPtr=&ChainWhip_HeavyImpact_3P_04}}
 {id=WavePlayer_8, properties {soundWaveAssetPtr=&ChainWhip_HeavyImpact_3P_01}}
 {id=WavePlayer_9, properties {soundWaveAssetPtr=&ChainWhip_HeavyImpact_3P_02}}
]
properties {
 bOverrideConcurrency=true
 maxDistance=10058.75
 soundClassObject=&Fort_WeaponMelee_PSM
 volumeMultiplier=0.9
 concurrencyOverrides {maxCount=3}
}
