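# Sound cue for picking up the Chain Whip item, classed as loot-collect audio with weapon pickup attenuation.
# A mixer layers the boosted main pickup wave with a quieter, evenly weighted random choice of three swing soul sweetener waves.
# Each layer passes through its own modulator.
# A concurrency override allows one instance at a time and blocks new ones while it plays.
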SoundCue /SunRoseChainWhipGameplay/Items/ChainWhip/Sound/ChainWhip_Pickup_Cue
root=Mixer_0
nodes [
 {id=Mixer_0, children [Modulator_5,Modulator_0], properties {inputVolume [3,0.5]}}
 {id=Modulator_0, children [Random_0]}
 {id=Modulator_5, children [WavePlayer_0]}
 {id=Random_0, children [WavePlayer_1,WavePlayer_2,WavePlayer_3], properties {weights [1,1,1]}}
 {id=WavePlayer_0, properties {soundWaveAssetPtr=&/ChainWhip_Pickup}}
 {id=WavePlayer_1, properties {soundWaveAssetPtr=&/ChainWhip_Swing_Soul_Sweetener_07}}
 {id=WavePlayer_2, properties {soundWaveAssetPtr=&/ChainWhip_Swing_Soul_Sweetener_04}}
 {id=WavePlayer_3, properties {soundWaveAssetPtr=&/ChainWhip_Swing_Soul_Sweetener_02}}
]
properties {
 attenuationSettings=&Fort_Weapon_Pickup_Att
 bOverrideConcurrency=true
 maxDistance=3500
 soundClassObject=&Fort_Loot_Collect
 concurrencyOverrides {maxCount=1, resolutionRule=PreventNew}
 concurrencySet [&PickupMaximum_Oldest]
}
