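# Sound cue for the idle loop of the Chain Whip pickup, using the shared loot ambient sound class and concurrency rule.
# A looping ChainWhip_DropIdle wave passes through an Enveloper that ramps volume up from near silence at the start.
# A quality-level root routes every tier to that same branch.
# The cue slightly raises pitch, lowers volume and references the default mono item ambient-loop attenuation asset.
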SoundCue /SunRoseChainWhipGameplay/Items/ChainWhip/Sound/ChainWhip_Pickup_IdleLoop_Cue
root=QualityLevel_0
nodes [
 {
  id=Enveloper_0
  children [WavePlayer_0]
  properties {pitchCurve {editorCurveData {keys [{time=0}]}}, volumeCurve {editorCurveData {keys [<at 0 0.05>,<at 0.55 1>]}}}
 }
 {id=QualityLevel_0, children [Enveloper_0,Enveloper_0,Enveloper_0,Enveloper_0]}
 {id=WavePlayer_0, properties {bLooping=true, soundWaveAssetPtr=&/ChainWhip_DropIdle}}
]
properties {
 attenuationSettings=/Game/Sounds/Attenuation/Fort_Item_Ambient_Loop_Mono_Default_Att
 maxDistance=1250
 pitchMultiplier=1.1
 soundClassObject=/Game/Sounds/FortSoundClassesAndModes/Fort_Loot_Ambient_Loop
 volumeMultiplier=0.6
 attenuationOverrides {distanceAlgorithm=NaturalSound, falloffDistance=3000, attenuationShapeExtents {x=1000}}
 concurrencySet [/Game/Sounds/Concurrency_Rules/Loot_Ambient_Conc]
}
