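# Sound cue for the Chain Whip projectile's in-air loop, spatialized by its own attenuation asset, in the melee weapon sound class with capped concurrency.
# A Mixer layers two random swing whooshes, pitched down to different depths, over a looping in-air wave.
# An Enveloper ramps the loop's volume up from silence, while the deeper whoosh runs through a Delay node left at default settings.
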SoundCue /SunRoseChainWhipGameplay/Items/ChainWhip/Sound/ChainWhip_Prj_InAir_Loop_Cue
root=Mixer_0
nodes [
 {id=Delay_0, children [Modulator_3]}
 {
  id=Enveloper_0
  children [WavePlayer_5]
  properties {pitchCurve {editorCurveData {keys [{time=0}]}}, volumeCurve {editorCurveData {keys [<at 0 0>,<at 1 1>]}}}
 }
 {id=Mixer_0, children [Modulator_1,Delay_0,Enveloper_0], properties {inputVolume [1.5,1.2,1.4]}}
 {id=Modulator_1, children [Random_0], properties {pitchMax=0.9, pitchMin=0.85}}
 {id=Modulator_3, children [Random_3], properties {pitchMax=0.6, pitchMin=0.5}}
 {
  id=Random_0
  children [WavePlayer_1, WavePlayer_2, WavePlayer_3, WavePlayer_4, WavePlayer_11]
  properties {bShouldExcludeFromBranchCulling=true, weights [1,1,1,1,1]}
 }
 {
  id=Random_3
  children [WavePlayer_0, WavePlayer_16, WavePlayer_17, WavePlayer_18, WavePlayer_28]
  properties {bShouldExcludeFromBranchCulling=true, weights [1,1,1,1,1]}
 }
 {id=WavePlayer_0, properties {soundWaveAssetPtr=&/ChainWhip_Swing_Whoosh_05}}
 {id=WavePlayer_1, properties {soundWaveAssetPtr=&ChainWhip_Swing_Whoosh_05}}
 {id=WavePlayer_11, properties {soundWaveAssetPtr=&/ChainWhip_Swing_Whoosh_01}}
 {id=WavePlayer_16, properties {soundWaveAssetPtr=&/ChainWhip_Swing_Whoosh_04}}
 {id=WavePlayer_17, properties {soundWaveAssetPtr=&/ChainWhip_Swing_Whoosh_03}}
 {id=WavePlayer_18, properties {soundWaveAssetPtr=&/ChainWhip_Swing_Whoosh_02}}
 {id=WavePlayer_2, properties {soundWaveAssetPtr=&ChainWhip_Swing_Whoosh_04}}
 {id=WavePlayer_28, properties {soundWaveAssetPtr=&ChainWhip_Swing_Whoosh_01}}
 {id=WavePlayer_3, properties {soundWaveAssetPtr=&ChainWhip_Swing_Whoosh_03}}
 {id=WavePlayer_4, properties {soundWaveAssetPtr=&ChainWhip_Swing_Whoosh_02}}
 {id=WavePlayer_5, properties {bLooping=true, soundWaveAssetPtr=&/ChainWhip_Prj_InAir_Loop}}
]
properties {
 attenuationSettings=&/ChainWhip_Prj_InAir_Loop_Att
 bOverrideConcurrency=true
 maxDistance=6000
 soundClassObject=&Fort_WeaponMelee
 volumeMultiplier=0.5
 concurrencyOverrides {maxCount=3}
}
