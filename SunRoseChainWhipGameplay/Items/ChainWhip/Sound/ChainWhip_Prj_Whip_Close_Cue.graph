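# Sound cue for the Chain Whip item's close projectile whip sound, classed as a ranged weapon projectile effect.
# A quality-level node sends its first two slots to a random pick of six whip waves and its last two to the final three of those waves.
# A neutral Modulator feeds the output at reduced volume.
# It references the shared bullet-whip attenuation asset and bullet whiz-by concurrency rule, with virtualization disabled.
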SoundCue /SunRoseChainWhipGameplay/Items/ChainWhip/Sound/ChainWhip_Prj_Whip_Close_Cue
root=Modulator_0
nodes [
 {id=Modulator_0, children [QualityLevel_0], properties {pitchMax=1, pitchMin=1, volumeMax=1, volumeMin=1}}
 {id=QualityLevel_0, children [Random_0,Random_0,Random_1,Random_1]}
 {
  id=Random_0
  children [WavePlayer_0,WavePlayer_5,WavePlayer_6,WavePlayer_7,WavePlayer_8,WavePlayer_9]
  properties {bShouldExcludeFromBranchCulling=true, weights [1,1,1,1,1,1]}
 }
 {id=Random_1, children [WavePlayer_7,WavePlayer_8,WavePlayer_9], properties {bShouldExcludeFromBranchCulling=true, weights [1,1,1]}}
 {id=WavePlayer_0, properties {soundWaveAssetPtr=&/ChainWhip_Whip_Close_01}}
 {id=WavePlayer_5, properties {soundWaveAssetPtr=&/ChainWhip_Whip_Close_02}}
 {id=WavePlayer_6, properties {soundWaveAssetPtr=&/ChainWhip_Whip_Close_03}}
 {id=WavePlayer_7, properties {soundWaveAssetPtr=&/ChainWhip_Whip_Close_04}}
 {id=WavePlayer_8, properties {soundWaveAssetPtr=&/ChainWhip_Whip_Close_05}}
 {id=WavePlayer_9, properties {soundWaveAssetPtr=&/ChainWhip_Whip_Close_06}}
]
properties {
 attenuationSettings=/Game/Athena/Sounds/Weapons/BulletWhips/bullet_whip_Att.Bullet_Whip_Att
 bBypassVolumeScaleForPriority=true
 maxDistance=2097152
 priority=0.5
 soundClassObject=&Fort_WeaponRanged_Projectile_Effects
 virtualizationMode=Disabled
 volumeMultiplier=0.4
 attenuationOverrides {bAttenuate=false, stereoSpread=0}
 concurrencySet [/Game/Sounds/Concurrency_Rules/Athena_BulletWhizBy_Conc]
}
