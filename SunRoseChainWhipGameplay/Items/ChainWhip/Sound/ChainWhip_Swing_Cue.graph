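# Sound cue for the chain whip's melee swing, layering two randomized whoosh sounds with close-range melee attenuation.
# A mixer combines a projectile-cast whoosh layer and a swing whoosh layer, each passing through its own modulator.
# In each layer, a quality-level node selects between random pools of five or three variations, shrinking the wave set on some quality tiers.
# A concurrency override caps simultaneous instances.
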SoundCue /SunRoseChainWhipGameplay/Items/ChainWhip/Sound/ChainWhip_Swing_Cue
root=Mixer_0
nodes [
 {id=Mixer_0, children [Modulator_2,Modulator_1], properties {inputVolume [1,1]}}
 {id=Modulator_1, children [QualityLevel_2], properties {pitchMax=1, pitchMin=1}}
 {id=Modulator_2, children [QualityLevel_1], properties {pitchMax=1}}
 {id=QualityLevel_1, children [Random_3,Random_3,Random_2,Random_2]}
 {id=QualityLevel_2, children [Random_5,Random_5,Random_6,Random_6]}
 {id=Random_2, children [WavePlayer_7,WavePlayer_8,WavePlayer_9], properties {bShouldExcludeFromBranchCulling=true, weights [1,1,1]}}
 {
  id=Random_3
  children [WavePlayer_5,WavePlayer_6,WavePlayer_7,WavePlayer_8,WavePlayer_9]
  properties {bShouldExcludeFromBranchCulling=true, weights [1,1,1,1,1]}
 }
 {
  id=Random_5
  children [WavePlayer_10, WavePlayer_11, WavePlayer_12, WavePlayer_13, WavePlayer_28]
  properties {bShouldExcludeFromBranchCulling=true, weights [1,1,1,1,1]}
 }
 {id=Random_6, children [WavePlayer_12, WavePlayer_13, WavePlayer_28], properties {bShouldExcludeFromBranchCulling=true, weights [1,1,1]}}
 {id=WavePlayer_10, properties {soundWaveAssetPtr=&/ChainWhip_Swing_Whoosh_01}}
 {id=WavePlayer_11, properties {soundWaveAssetPtr=&/ChainWhip_Swing_Whoosh_02}}
 {id=WavePlayer_12, properties {soundWaveAssetPtr=&/ChainWhip_Swing_Whoosh_03}}
 {id=WavePlayer_13, properties {soundWaveAssetPtr=&/ChainWhip_Swing_Whoosh_04}}
 {id=WavePlayer_28, properties {soundWaveAssetPtr=&/ChainWhip_Swing_Whoosh_05}}
 {id=WavePlayer_5, properties {soundWaveAssetPtr=&/ChainWhip_Prj_Cast_Whoosh_01}}
 {id=WavePlayer_6, properties {soundWaveAssetPtr=&/ChainWhip_Prj_Cast_Whoosh_02}}
 {id=WavePlayer_7, properties {soundWaveAssetPtr=&/ChainWhip_Prj_Cast_Whoosh_03}}
 {id=WavePlayer_8, properties {soundWaveAssetPtr=&/ChainWhip_Prj_Cast_Whoosh_04}}
 {id=WavePlayer_9, properties {soundWaveAssetPtr=&/ChainWhip_Prj_Cast_Whoosh_05}}
]
properties {
 attenuationSettings=/Game/Sounds/Attenuation/Fort_Weapon_Melee_Attack_Close_Range_Stereo_Att
 bOverrideConcurrency=true
 maxDistance=10058.75
 pitchMultiplier=1.05
 soundClassObject=&Fort_WeaponMelee_PSM
 volumeMultiplier=1
 concurrencyOverrides {maxCount=4}
}
