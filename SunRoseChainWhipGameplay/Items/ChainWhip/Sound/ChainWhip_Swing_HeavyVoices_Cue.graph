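# Sound cue that plays a randomized soul-sweetener voice layer for heavy chain whip melee swings.
# A quality-level switch feeds a modulator, with the first two slots picking from nine waves and the last two from only the final three of those.
# Both random nodes weight waves equally and are excluded from branch culling.
# It uses close-range melee attenuation, a melee weapon sound class and an overridden concurrency limit.
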SoundCue /SunRoseChainWhipGameplay/Items/ChainWhip/Sound/ChainWhip_Swing_HeavyVoices_Cue
root=Modulator_0
nodes [
 {id=Modulator_0, children [QualityLevel_0]}
 {id=QualityLevel_0, children [Random_4,Random_4,Random_1,Random_1]}
 {id=Random_1, children [WavePlayer_25, WavePlayer_26, WavePlayer_1], properties {bShouldExcludeFromBranchCulling=true, weights [1,1,1]}}
 {
  id=Random_4
  children [WavePlayer_19, WavePlayer_20, WavePlayer_21, WavePlayer_22, WavePlayer_23, WavePlayer_24, WavePlayer_25, WavePlayer_26, WavePlayer_1]
  properties {bShouldExcludeFromBranchCulling=true, weights [1,1,1,1,1,1,1,1,1]}
 }
 {id=WavePlayer_1, properties {soundWaveAssetPtr=&/ChainWhip_Swing_Soul_Sweetener_09}}
 {id=WavePlayer_19, properties {soundWaveAssetPtr=&/ChainWhip_Swing_Soul_Sweetener_01}}
 {id=WavePlayer_20, properties {soundWaveAssetPtr=&/ChainWhip_Swing_Soul_Sweetener_02}}
 {id=WavePlayer_21, properties {soundWaveAssetPtr=&/ChainWhip_Swing_Soul_Sweetener_03}}
 {id=WavePlayer_22, properties {soundWaveAssetPtr=&/ChainWhip_Swing_Soul_Sweetener_04}}
 {id=WavePlayer_23, properties {soundWaveAssetPtr=&/ChainWhip_Swing_Soul_Sweetener_05}}
 {id=WavePlayer_24, properties {soundWaveAssetPtr=&/ChainWhip_Swing_Soul_Sweetener_06}}
 {id=WavePlayer_25, properties {soundWaveAssetPtr=&/ChainWhip_Swing_Soul_Sweetener_07}}
 {id=WavePlayer_26, properties {soundWaveAssetPtr=&/ChainWhip_Swing_Soul_Sweetener_08}}
]
properties {
 attenuationSettings=/Game/Sounds/Attenuation/Fort_Weapon_Melee_Attack_Close_Range_Stereo_Att
 bOverrideConcurrency=true
 maxDistance=10058.75
 soundClassObject=&Fort_WeaponMelee_PSM
 volumeMultiplier=1.35
 concurrencyOverrides {maxCount=3}
}
