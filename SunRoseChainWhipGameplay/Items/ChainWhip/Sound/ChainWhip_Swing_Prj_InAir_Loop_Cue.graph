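# Sound cue for the chain whip projectile's in-air swing, layering a whoosh with a looping flight sound.
# A Mixer combines a randomly picked, modulated whoosh from five variants with the in-air loop, which an Enveloper fades in.
# A Branch on `islocal` plays that mix at quarter volume when true or unset, and unreduced when false.
# It uses short-range attenuation, the melee weapon sound class and a small concurrency cap.
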SoundCue /SunRoseChainWhipGameplay/Items/ChainWhip/Sound/ChainWhip_Swing_Prj_InAir_Loop_Cue
root=Branch_0
nodes [
 {id=Branch_0, children [Modulator_3,Mixer_0,Modulator_3], properties {boolParameterName=islocal}}
 {
  id=Enveloper_0
  children [WavePlayer_5]
  properties {pitchCurve {editorCurveData {keys [{time=0}]}}, volumeCurve {editorCurveData {keys [<at 0 0>,<at 1 1>]}}}
 }
 {id=Mixer_0, children [null,Modulator_1,null,Enveloper_0,null], properties {inputVolume [1,1.5,2,1,1]}}
 {id=Modulator_1, children [Random_0]}
 {id=Modulator_3, children [Mixer_0], properties {pitchMax=1, pitchMin=1, volumeMax=0.25, volumeMin=0.25}}
 {
  id=Random_0
  children [WavePlayer_1, WavePlayer_2, WavePlayer_3, WavePlayer_4, WavePlayer_11]
  properties {bShouldExcludeFromBranchCulling=true, weights [1,1,1,1,1]}
 }
 {id=WavePlayer_1, properties {soundWaveAssetPtr=&/ChainWhip_Swing_Whoosh_05}}
 {id=WavePlayer_11, properties {soundWaveAssetPtr=&/ChainWhip_Swing_Whoosh_01}}
 {id=WavePlayer_2, properties {soundWaveAssetPtr=&/ChainWhip_Swing_Whoosh_04}}
 {id=WavePlayer_3, properties {soundWaveAssetPtr=&/ChainWhip_Swing_Whoosh_03}}
 {id=WavePlayer_4, properties {soundWaveAssetPtr=&/ChainWhip_Swing_Whoosh_02}}
 {id=WavePlayer_5, properties {bLooping=true, soundWaveAssetPtr=&/ChainWhip_Prj_InAir_Loop}}
]
properties {
 attenuationSettings=&/ChainWhip_Prj_InAir_Loop_Short_Att
 bDebug=true
 bOverrideConcurrency=true
 maxDistance=4400
 soundClassObject=&Fort_WeaponMelee
 volumeMultiplier=0.7
 concurrencyOverrides {maxCount=3}
}
