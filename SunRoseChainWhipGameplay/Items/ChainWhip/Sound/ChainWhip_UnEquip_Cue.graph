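# Sound cue for unequipping the Chain Whip, built from reused Ghost Pistol trail sounds.
# A Random node picks one of three equally weighted waves, an Enveloper fades its volume out along a short curve, and a Modulator at the root outputs the result.
# It plays quietly in the weapon foley class with shared weapon attenuation, and overrides concurrency to cap simultaneous plays per owner with a retrigger time.
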SoundCue /SunRoseChainWhipGameplay/Items/ChainWhip/Sound/ChainWhip_UnEquip_Cue
root=Modulator_0
nodes [
 {
  id=Enveloper_0
  children [Random_1]
  properties {pitchCurve {editorCurveData {keys [{time=0}]}}, volumeCurve {editorCurveData {keys [<at 0 1>,<at 0.25 0.75>,<at 2 0>]}}}
 }
 {id=Modulator_0, children [Enveloper_0]}
 {id=Random_1, children [WavePlayer_0,WavePlayer_5,WavePlayer_6], properties {weights [1,1,1]}}
 {id=WavePlayer_0, properties {soundWaveAssetPtr=/Game/Sounds/Fort_Weapons/pistol/Ghost/SFX_GhostPistol_GhostTrail_03}}
 {id=WavePlayer_5, properties {soundWaveAssetPtr=/Game/Sounds/Fort_Weapons/pistol/Ghost/SFX_GhostPistol_GhostTrail_02}}
 {id=WavePlayer_6, properties {soundWaveAssetPtr=/Game/Sounds/Fort_Weapons/pistol/Ghost/SFX_GhostPistol_GhostTrail_01}}
]
properties {
 attenuationSettings=/Game/Sounds/Attenuation/Fort_Weapon_AB_Ambient_Athena_Att
 bOverrideConcurrency=true
 maxDistance=10000
 soundClassObject=&Fort_WeaponFoley
 volumeMultiplier=0.376
 attenuationOverrides {distanceAlgorithm=NaturalSound, falloffDistance=3000, attenuationShapeExtents {x=1000}}
 concurrencyOverrides {bLimitToOwner=true, maxCount=2, retriggerTime=0.2}
}
