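# Sound cue for the chain whip's landing foley, playing one of four random metal chain jingles.
# An envelope fades the volume out, feeding two attenuation paths that share the same random source.
# A branch on "islocal" picks ChainWhip attenuation when true, otherwise basic foley attenuation with a slight volume boost and pitch variation.
# The cue plays quietly in the weapon foley sound class, with concurrency overrides capping it at one instance.
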SoundCue /SunRoseChainWhipGameplay/Items/ChainWhip/Sound/Foley_ChainWhip_Land_Cue
root=Branch_0
nodes [
 {id=Attenuation_0, children [Enveloper_0], properties {attenuationSettings=&/ChainWhip_Att}}
 {
  id=Attenuation_1
  children [Enveloper_0]
  properties {attenuationSettings=&Foley_Attenuation_Basic_Att}
 }
 {id=Branch_0, children [Attenuation_0, Modulator_0, Modulator_0], properties {boolParameterName=islocal}}
 {
  id=Enveloper_0
  children [Random_2]
  properties {
   pitchCurve {editorCurveData {keys [{time=0}]}}
   volumeCurve {editorCurveData {keys [<at 0 1>,<at 0.2578 0.8139>,<at 0.5473 0.2639>,<at 1 0>]}}
  }
 }
 {id=Modulator_0, children [Attenuation_1], properties {pitchMax=1, volumeMax=1.2, volumeMin=1.2}}
 {id=Random_2, children [WavePlayer_12, WavePlayer_13, WavePlayer_14, WavePlayer_15], properties {weights [1,1,1,1]}}
 {id=WavePlayer_12, properties {soundWaveAssetPtr=/Game/Sounds/Trans_AmbientNonLoop/Metal/ChainDingleA01}}
 {id=WavePlayer_13, properties {soundWaveAssetPtr=/Game/Sounds/Trans_AmbientNonLoop/Metal/ChainDingleA02}}
 {id=WavePlayer_14, properties {soundWaveAssetPtr=/Game/Sounds/Trans_AmbientNonLoop/Metal/ChainDingleA03}}
 {id=WavePlayer_15, properties {soundWaveAssetPtr=/Game/Sounds/Trans_AmbientNonLoop/Metal/ChainDingleA04}}
]
properties {
 maxDistance=14000
 soundClassObject=&Fort_WeaponFoley
 volumeMultiplier=0.125
 concurrencyOverrides {maxCount=1}
}
